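# Idle body effect for the LooseCreep female cosmetic: blue spinning gears and glows that follow two skeletal mesh sockets.
# Gear spawns two burst groups of mesh particles, one per socket, with paired gears counter-rotating and sliding on layered sine offsets.
# Glows layers sprite, fake-light and darken attributes per socket, scaled by platform and time of day.
# User.bShowBackFX zeroes one socket's alpha, and front-end modifiers adjust brightness.
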NiagaraSystem /BRCosmetics/Characters/Player/Female/Medium/Bodies/F_MED_LooseCreep/FX/NS_LooseCreep_Body_Idle
dependencies {
 usedDataInterfaces [
  &NiagaraDataInterfaceArrayFloat
  &NiagaraDataInterfaceArrayInt32
  &NiagaraDataInterfaceCurve
  &NiagaraDataInterfacePlatformSet
  &NiagaraDataInterfaceArrayFloat3
  &NiagaraDataInterfaceSkeletalMesh
 ]
 usedDynamicInputs [
  &FloatFromCurve
  &UniformRangedFloat
  &FNTimeOfDayScalar
 ]
 usedModules [
  &self:SystemSpawnScript.SystemScriptSource.SystemScriptGraph.NiagaraNodeAssignment_0.SetVariables_#1
  &self:SystemSpawnScript.SystemScriptSource.SystemScriptGraph.NiagaraNodeAssignment_1.SetVariables_#2
  &self:SystemSpawnScript.SystemScriptSource.SystemScriptGraph.NiagaraNodeAssignment_2.SetVariables_#3
  &SystemState
 ]
}
emitters {
 Gear {
  emitterUpdateScript [
   &EmitterState(Life Cycle Mode=::NewEnumerator1, Loop Behavior=::NewEnumerator1, Loop Duration Mode=::NewEnumerator1, UseLoopDelay=false)
   &SpawnBurst_Instantaneous(Spawn Count=2, Spawn Group=1, Use Spawn Probability=false, UseLoopCountLimit=false) {module=SpawnBurst_Instantaneous001}
   &SpawnBurst_Instantaneous(Spawn Count=2, Use Spawn Probability=false, UseLoopCountLimit=false)
  ]
  particleSpawnScript [
   &PartitionParticles()
   &InitializeParticle(
    Lifetime=0.8
    Mesh Scale=&Multiply_VectorByFloat(
     Vector=<1 1 1>
     Float=&Multiply_Float(
      A=&MergeHelper_FloatSwitch(0=0.04, 1=0.025, 2=?, 3=?, Index=$Particles.PartitionParticles.Partition, Num=::NewEnumerator0)
      B=&MergeHelper_FloatSwitch(0=1, 1=0.75, 2=?, 3=?, Index=$Emitter.SpawnGroup, Num=::NewEnumerator0)
     )
    )
    Mesh Scale Mode=::NewEnumerator4
    Sprite Size Mode=::NewEnumerator3
    UsePositionOffset=false
    Write Material Random=true
   )
   &CameraOffset(Camera Offset Amount=1)
   &InitialMeshOrientation(
    Mesh Orientation Mode=::NewEnumerator7
    Rotation=&MergeHelper_VectorSwitch(0=<0 0 0.25>, 1=<0 0 -0.25>, 2=?, 3=?, Index=$Emitter.SpawnGroup, Num=::NewEnumerator0)
    Use Rotation Vector=true
   )
   /SetVariables(Particles.VisibilityTag=$Emitter.SpawnGroup) {isSetParameters=true}
  ]
  particleUpdateScript [
   &ParticleState(Kill Particles When Lifetime Has Elapsed=false, Let Infinitely Lived Particles Die When Emitter Deactivates=true)
   &SkeletalMeshLocation(
    Bone / Socket Index=&SelectIntFromArray(Direct Array Index=$Particles.VisibilityTag, Int Selection Array=DataInterface() {propertyValues {IntData [1,0]}})
    MaskSpawnGroup=false
    Mesh Sampling Type=::NewEnumerator5
    Sampled Position Offset=&Multiply_VectorByFloat(
     Vector=<0 1 0>
     Float=&Multiply_Float(
      A=&Divide_Float(
       A=&Add_Float(
        A=&Sine(Normalized Angle=$Emitter.Age, Period=6, Scale=1, Bias=0)
        B=&Multiply_Float(
         A=0.05
         B=&Sine(Normalized Angle=&Multiply_Float(A=$Emitter.Age, B=3), Period=6, Scale=1, Bias=0)
        )
       )
       B=&Subtract_Float(A=1, B=0.05)
      )
      B=&MergeHelper_FloatSwitch(
       0=&MergeHelper_FloatSwitch(0=1, 1=-1, 2=?, 3=?, Index=$Particles.VisibilityTag, Num=::NewEnumerator0)
       1=&MergeHelper_FloatSwitch(0=-1, 1=1, 2=?, 3=?, Index=$Particles.VisibilityTag, Num=::NewEnumerator0)
       2=?
       3=?
       Index=$Particles.PartitionParticles.Partition
       Num=::NewEnumerator0
      )
     )
    )
    Skeletal Mesh=$System.Skeletal Mesh
    Socket Sampling Mode=::NewEnumerator5
   )
   /SetVariables(Particles.MeshOrientation=&MultiplyQuaternion(
     Quaternion A=$Particles.MeshOrientation
     Quaternion B=&MakeQuaternion(
      Quaternion From=0
      Axis=<1 0 0>
      Angle=&Multiply_Float(
       A=$Engine.Time
       B=&MergeHelper_FloatSwitch(0=90, 1=-90, 2=?, 3=?, Index=$Particles.PartitionParticles.Partition, Num=::NewEnumerator0)
      )
      From=?
      To=?
      Angle Type=::NewEnumerator0
      Normalize Vectors=false
     )
    )) {isSetParameters=true}
   &ScaleColor(
    Scale Alpha=&MergeHelper_FloatSwitch(
     0=&MakeCustomFloatFromBool(Bool=$User.bShowBackFX, True Float=1, False Float=0)
     1=1
     2=?
     3=?
     Index=$Particles.VisibilityTag
     Num=::NewEnumerator0
    )
    Scale RGB=&FN_FrontEndVector3Modifier(
     Gameplay=&MakeVectorFromLinearColorRGB(Color=<rgba 0 0.48 1.2 1>)
     Front End=&MakeVectorFromLinearColorRGB(Color=<rgba 0 0.8 2 1>)
    )
    ScaleA=true
    ScaleRGB=true
   )
  ]
  rendererClasses [&NiagaraMeshRendererProperties]
  renderers [
   {rendererClass=&NiagaraMeshRendererProperties, rendererIndex=0}
   {rendererClass=&NiagaraMeshRendererProperties, rendererIndex=1}
  ]
 }
 Glows {
  emitterUpdateScript [
   &EmitterState(Life Cycle Mode=::NewEnumerator1, Loop Behavior=::NewEnumerator1, Loop Duration Mode=::NewEnumerator1, UseLoopDelay=false)
   &SpawnBurst_Instantaneous(Spawn Count=2, Use Spawn Probability=false, UseLoopCountLimit=false)
  ]
  particleSpawnScript [
   &PartitionParticles()
   &InitializeParticle(
    Color=<rgba 0 1.2 3 1>
    Position Offset=<-4 0 0>
    Sprite Rotation Mode=::NewEnumerator2
    Sprite Size Mode=::NewEnumerator3
    Uniform Sprite Size=&SelectFloatFromArray(Direct Array Index=$Particles.PartitionParticles.Partition, Float Selection Array=DataInterface() {propertyValues {FloatData [6,5]}})
    UsePositionOffset=false
    Write Material Random=true
   )
  ]
  particleUpdateScript [
   &ParticleState(Kill Particles When Lifetime Has Elapsed=false, Let Infinitely Lived Particles Die When Emitter Deactivates=true)
   &SkeletalMeshLocation(
    Bone / Socket Index=&SelectIntFromArray(Direct Array Index=$Particles.PartitionParticles.Partition, Int Selection Array=DataInterface() {propertyValues {IntData [1,0]}})
    MaskSpawnGroup=false
    Mesh Sampling Type=::NewEnumerator5
    Sampled Position Offset=&MergeHelper_VectorSwitch(0=<0 1.5 0>, 1=<0 -1.5 0>, 2=?, 3=?, Index=$Particles.PartitionParticles.Partition, Num=::NewEnumerator0)
    Skeletal Mesh=$System.Skeletal Mesh
    Socket Sampling Mode=::NewEnumerator5
   )
   &SolveForcesAndVelocity(Clamp Velocity=false, Limit Acceleration=false)
   &DynamicMaterialParameters(Write Parameter Index 0=::NewEnumerator0)
   &CameraOffset(Camera Offset Amount=2)
   &FN_Light_Attributes(
    Color=<rgba 0 0.3478 1 1>
    Create Color Attribute=true
    Create Exponent Attribute=true
    Create Radius Attribute=true
    Enable Particle Light Enable=true
    Exponent=7
    Light Enabled=true
    Radius=50
    Scale Color=5
   )
   /SetVariables(
    Particles.ColorFakeLight:LinearColor=&Multiply_LinearColorByFloat(
     Linear Color=&Multiply_LinearColorByFloat(
      Linear Color=<rgba 0 0.1391 1 0.4>
      Float=&FN_PickFloatDependOnPlatform(TargetPlatformActive=0.6, TargetPlatformNoActive=0.4)
     )
     Float=&MergeHelper_FloatSwitch(
      0=&MakeCustomFloatFromBool(Bool=$User.bShowBackFX, True Float=1, False Float=0)
      1=1
      2=?
      3=?
      Index=$Particles.PartitionParticles.Partition
      Num=::NewEnumerator0
     )
    )
    Particles.FakeLightPosition:NiagaraPosition=&AddVectorToPosition(
     Position=$Particles.SkeletalMeshLocation.SampledPosition
     Vector=&SelectVectorFromArray(
      Direct Array Index=$Particles.PartitionParticles.Partition
      Vector Selection Array=DataInterface() {propertyValues {FloatData [<0 0 0>, <0 -2 0>]}}
     )
    )
    Particles.FakeLightSize:Vector2f=&Vector2DFromFloat(Value=&SelectFloatFromArray(Direct Array Index=$Particles.PartitionParticles.Partition, Float Selection Array=DataInterface() {propertyValues {FloatData [30,40]}}))
   ) {isSetParameters=true}
   &ScaleColor(
    Scale Alpha=&Multiply_Float(
     A=&Multiply_Float(
      A=&FN_PickFloatDependOnPlatform(TargetPlatformActive=1, TargetPlatformNoActive=0.5)
      B=$System.TOD Scalar
     )
     B=&MergeHelper_FloatSwitch(
      0=&MakeCustomFloatFromBool(Bool=$User.bShowBackFX, True Float=1, False Float=0)
      1=1
      2=?
      3=?
      Index=$Particles.PartitionParticles.Partition
      Num=::NewEnumerator0
     )
    )
    ScaleA=true
    ScaleRGB=true
   )
   /SetVariables(
    Particles.Darken Color:LinearColor=&MakeLinearColorFromVectorAndFloat(
     "Vector (RGB)"=&MakeVectorFromLinearColorRGB(Color=<rgba 0 0.4522 1 1>)
     "Float (Alpha)"=&Multiply_Float(
      A=&Multiply_Float(
       A=&FN_FrontEndFloatModifier(
        Gameplay=&FN_PickFloatDependOnPlatform(TargetPlatformActive=0.15, TargetPlatformNoActive=0.1)
        Front End=0.2
       )
       B=$System.TOD Scalar
      )
      B=&MergeHelper_FloatSwitch(
       0=&MakeCustomFloatFromBool(Bool=$User.bShowBackFX, True Float=1, False Float=0)
       1=1
       2=?
       3=?
       Index=$Particles.PartitionParticles.Partition
       Num=::NewEnumerator0
      )
     )
    )
    Particles.Darken Size:Vector2f=&Multiply_Vector2D_ByFloat(
     Float=&MergeHelper_FloatSwitch(0=6, 1=6.0565, 2=?, 3=?, Index=$Particles.PartitionParticles.Partition, Num=::NewEnumerator0)
     Vector2D=$Particles.SpriteSize
    )
   ) {isSetParameters=true}
  ]
  rendererClasses [&NiagaraSpriteRendererProperties]
  renderers [
   {rendererClass=&NiagaraSpriteRendererProperties}
   {rendererClass=&NiagaraSpriteRendererProperties, rendererIndex=1}
   {rendererClass=&NiagaraSpriteRendererProperties, rendererIndex=2}
  ]
 }
}
summary {
 systemName=NS_LooseCreep_Body_Idle
 emitters [
  {emitterName=Glows, rendererClasses [&NiagaraSpriteRendererProperties]}
  {emitterName=Gear, rendererClasses [&NiagaraMeshRendererProperties]}
 ]
 userVariables [
  User.bIsFrontEnd {type=NiagaraBool}
  User.bIsFrontEndPreview {type=NiagaraBool}
  User.bisHeadHidden {type=NiagaraBool}
  User.bShowBackFX {type=NiagaraBool, defaultValue=true}
  User.IsDriving {type=NiagaraBool}
  User.IsSkydiving {type=NiagaraBool}
 ]
}
