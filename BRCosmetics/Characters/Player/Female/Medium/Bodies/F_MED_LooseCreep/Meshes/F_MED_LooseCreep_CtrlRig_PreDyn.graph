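# Pre-dynamics Control Rig for the female medium LooseCreep body cosmetic, adding corrective motion to deform, collar, buckle and strap bones.
# Smoothed thigh pose readers drive glute and groin offsets around a shared DeformRig function, then front strap bases aim at the thighs.
# Collar and buckle bones follow weighted multi-parent constraints.
# Head-direction dot products rotate collar bones, while clavicle and upper-arm pose readers layer buckle and collar fixes, including a crouch variant.
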ControlRigBlueprint /BRCosmetics/Characters/Player/Female/Medium/Bodies/F_MED_LooseCreep/Meshes/F_MED_LooseCreep_CtrlRig_PreDyn
graphs [
 CollapseNode_ContainedGraph {nodes [
   RigVMFunctionEntryNode()
   Multiply = RigVMAggregateNode(
    C=&RigUnit_GetTransform(Item=(Type=Bone,Name="neck_01"), Space=LocalSpace, bInitial=False).Transform
    A=&RigUnit_GetTransform(Item=(Type=Bone,Name="head_end"), Space=LocalSpace, bInitial=False).Transform
    D=&RigUnit_GetTransform(Item=(Type=Bone,Name="neck_02"), Space=LocalSpace, bInitial=False).Transform
    B=&RigUnit_GetTransform(Item=(Type=Bone,Name="head"), Space=LocalSpace, bInitial=False).Transform
   )
   Unit = /Script/RigVM.RigVMFunction_MathVectorUnit(Value=@Multiply.Result.Translation) {is=struct RigVMUnitNode, template="Unit::Execute(in Value,out Result)"}
   RigVMFunctionReturnNode(Result=@Unit.Result)
  ]}
 Multiply_1_ContainedGraph {nodes [
   n1 = RigVMFunctionEntryNode()
   Multiply = &RigVMFunction_MathTransformMul(A=@n1.A, B=@n1.B) {template="Multiply::Execute(in A,in B,out Result)"}
   RigVMFunctionReturnNode(Result=&RigVMFunction_MathTransformMul(
     B=@n1.D
     A=&RigVMFunction_MathTransformMul(B=@n1.C, A=@Multiply.Result) {template="Multiply::Execute(in A,in B,out Result)"}.Result
    ) {template="Multiply::Execute(in A,in B,out Result)"}.Result)
  ]}
 New Function_ContainedGraph {nodes [
   n1 = RigVMFunctionEntryNode()
   Dot__ = /Script/RigVM.RigVMFunction_MathVectorDot(
    B=&RigVMFunction_MathVectorUnit(Value=&RigUnit_GetTransform(Item=(Type=Bone,Name="collar_09"), Space=LocalSpace, bInitial=False, Item.Name=@n1.BoneIn.Name).Transform.Translation) {template="Unit::Execute(in Value,out Result)"}.Result
    A=RigVMVariableNode() {variable=head_vector}.Value
   ) {is=struct RigVMUnitNode, template="Dot,|::Execute(in A,in B,out Result)"}
   n2 = &RigVMFunction_MathDoubleRemap(TargetMaximum=1, TargetMinimum=0, Value=@Dot__.Result, SourceMinimum=@n1.s_min, SourceMaximum=@n1.s_max)
   Multiply = &RigVMFunction_MathDoubleMul(A=@n1.angle, B=@n2.Result)
   > &RigUnit_ModifyTransforms(
    ItemToModify=((Item=(Type=Bone,Name="None"),Transform=(Rotation=(X=0,Y=0,Z=0,W=1),Translation=(X=0,Y=0,Z=0),Scale3D=(X=1,Y=1,Z=1))))
    Weight=1
    WeightMinimum=0
    ItemToModify.0.Transform.Rotation=&RigVMFunction_MathQuaternionFromEuler(Euler=(X=0,Y=0,Z=0), Euler.Z=@Multiply.Result).Result
    ItemToModify.0.Item.Name=@n1.BoneIn.Name
   )
   > RigVMFunctionReturnNode(DotAlpha=@n2.Result, BoneOut=@n1.BoneIn.Name)
  ]}
 RigVMFunctionLibrary {nodes [DotRot_CollarHead = RigVMCollapseNode(BoneIn=(Type=None,Name="None"), ExecuteContext=())]}
 RigVMFunction_Sequence_2_ContainedGraph {nodes [
   n1 = RigVMFunctionEntryNode()
   n2 = &RigVMFunction_Sequence(ExecuteContext=@n1.ExecuteContext)
   n3 = &RigVMFunction_Sequence(ExecuteContext=@n2.B)
   RigVMFunctionReturnNode(A=@n2.A, B=@n3.A, C=@n3.B)
  ]}
 RigVMFunction_Sequence_3_ContainedGraph {nodes [
   n1 = RigVMFunctionEntryNode()
   n2 = &RigVMFunction_Sequence(ExecuteContext=@n1.ExecuteContext)
   n3 = &RigVMFunction_Sequence(ExecuteContext=@n2.B)
   RigVMFunctionReturnNode(B=@n3.A, C=@n3.B, A=@n2.A)
  ]}
 RigVMFunction_Sequence_4_ContainedGraph {nodes [
   n1 = RigVMFunctionEntryNode()
   n2 = &RigVMFunction_Sequence(ExecuteContext=@n1.ExecuteContext)
   n3 = &RigVMFunction_Sequence(ExecuteContext=@n2.B)
   RigVMFunctionReturnNode(B=@n3.A, C=@n3.B, A=@n2.A)
  ]}
 RigVMFunction_Sequence_5_ContainedGraph {nodes [
   n1 = RigVMFunctionEntryNode()
   n2 = &RigVMFunction_Sequence(ExecuteContext=@n1.ExecuteContext)
   n3 = &RigVMFunction_Sequence(ExecuteContext=@n2.B)
   n4 = &RigVMFunction_Sequence(ExecuteContext=@n3.B)
   RigVMFunctionReturnNode(B=@n3.A, C=@n4.A, D=@n4.B, A=@n2.A)
  ]}
 RigVMFunction_Sequence_6_ContainedGraph {nodes [
   n1 = RigVMFunctionEntryNode()
   n2 = &RigVMFunction_Sequence(ExecuteContext=@n1.ExecuteContext)
   n3 = &RigVMFunction_Sequence(ExecuteContext=@n2.B)
   n4 = &RigVMFunction_Sequence(ExecuteContext=@n3.B)
   n5 = &RigVMFunction_Sequence(ExecuteContext=@n4.B)
   n6 = &RigVMFunction_Sequence(ExecuteContext=@n5.B)
   RigVMFunctionReturnNode(A=@n2.A, B=@n3.A, C=@n4.A, D=@n5.A, E=@n6.A, F=@n6.B)
  ]}
 RigVMModel {nodes [
   RigUnit_BeginExecution = &RigUnit_BeginExecution()
   RigVMFunction_Sequence_1 = RigVMAggregateNode(ExecuteContext=@RigUnit_BeginExecution.ExecutePin)
   RigVMFunction_Sequence = &RigVMFunction_Sequence(ExecuteContext=@RigVMFunction_Sequence_1.A)
   n1 = &RigUnit_SphericalPoseReader(
    ActiveRegionScaleFactors=(PositiveWidth=3.5,NegativeWidth=0.5,PositiveHeight=3,NegativeHeight=3)
    ActiveRegionSize=0.1
    Debug=(bDrawDebug=false,bDraw2D=false,DebugScale=25,DebugSegments=20,DebugThickness=0.25)
    DriverAxis=(X=1,Y=0,Z=0)
    DriverItem=(Type=Bone,Name="thigh_r")
    FalloffRegionScaleFactors=(PositiveWidth=1,NegativeWidth=1,PositiveHeight=1,NegativeHeight=1)
    FalloffSize=0.45
    FlipHeightScaling=false
    FlipWidthScaling=true
    OptionalParentItem=(Type=Bone)
    RotationOffset=(X=-90,Y=0,Z=0)
    ExecutePin=@RigVMFunction_Sequence.A
   )
   AlphaInterp = &RigVMFunction_AlphaInterp(InterpSpeedDecreasing=10, InterpSpeedIncreasing=10, Scale=1, bClampResult=False, Value=@n1.OutputParam)
   > RigVMVariableNode(Value=@AlphaInterp.Result) {variable=AlphaSPR_thigh_R}
   n2 = &RigUnit_SphericalPoseReader(
    ActiveRegionScaleFactors=(PositiveWidth=3.5,NegativeWidth=0.5,PositiveHeight=3,NegativeHeight=3)
    ActiveRegionSize=0.1
    Debug=(bDrawDebug=false,DebugScale=25,DebugSegments=20,DebugThickness=0.25)
    DriverAxis=(X=-1,Y=0,Z=0)
    DriverItem=(Type=Bone,Name="thigh_l")
    FalloffRegionScaleFactors=(PositiveWidth=1,NegativeWidth=1,PositiveHeight=1,NegativeHeight=1)
    FalloffSize=0.45
    OptionalParentItem=(Type=Bone)
    RotationOffset=(X=90,Y=0,Z=0)
    ExecutePin=@RigVMFunction_Sequence.B
   )
   n3 = &RigVMFunction_AlphaInterp(InterpSpeedDecreasing=2, InterpSpeedIncreasing=2, Scale=1, bClampResult=False, Value=@n2.OutputParam)
   > RigVMVariableNode(Value=@n3.Result) {variable=AlphaSPR_thigh_L}
   n4 = &RigVMFunction_Sequence(ExecuteContext=@RigVMFunction_Sequence_1.B)
   n5 = &RigUnit_SphericalPoseReader(
    ActiveRegionScaleFactors=(PositiveWidth=3.5,NegativeWidth=0.5,PositiveHeight=3,NegativeHeight=3)
    ActiveRegionSize=0.05
    Debug=(bDrawDebug=true,bDraw2D=false,DebugScale=25,DebugSegments=20,DebugThickness=0.25)
    DriverAxis=(X=1,Y=0,Z=0)
    DriverItem=(Type=Bone,Name="thigh_r")
    FalloffRegionScaleFactors=(PositiveWidth=1,NegativeWidth=1,PositiveHeight=1,NegativeHeight=1)
    FalloffSize=0.2
    FlipHeightScaling=false
    FlipWidthScaling=true
    OptionalParentItem=(Type=Bone)
    RotationOffset=(X=0,Y=0,Z=0)
    ExecutePin=@n4.A
   )
   n6 = &RigVMFunction_AlphaInterp(InterpSpeedDecreasing=10, InterpSpeedIncreasing=10, Scale=1, bClampResult=False, Value=@n5.OutputParam)
   > RigVMVariableNode(Value=@n6.Result) {variable=AlphaSPR_thigh_side_R}
   n7 = &RigUnit_SphericalPoseReader(
    ActiveRegionScaleFactors=(PositiveWidth=3.5,NegativeWidth=0.5,PositiveHeight=3,NegativeHeight=3)
    ActiveRegionSize=0.05
    Debug=(bDrawDebug=true,DebugScale=25,DebugSegments=20,DebugThickness=0.25)
    DriverAxis=(X=-1,Y=0,Z=0)
    DriverItem=(Type=Bone,Name="thigh_l")
    FalloffRegionScaleFactors=(PositiveWidth=1,NegativeWidth=1,PositiveHeight=1,NegativeHeight=1)
    FalloffSize=0.2
    OptionalParentItem=(Type=Bone)
    RotationOffset=(X=180,Y=0,Z=0)
    ExecutePin=@n4.B
   )
   n8 = &RigVMFunction_AlphaInterp(InterpSpeedDecreasing=10, InterpSpeedIncreasing=10, Scale=1, bClampResult=False, Value=@n7.OutputParam)
   > RigVMVariableNode(Value=@n8.Result) {variable=AlphaSPR_thigh_side_L}
   FN_DeformRig_FNC = RigVMFunctionReferenceNode(
    DeformRig_ Curve="(deformRig_l_upperarm_up_alpha_72_#1=False,deformRig_r_upperarm_up_alpha_73_#2=False,deformRig_l_upperarm_fr_alpha_74_#3=False,deformRig_r_upperarm_fr_alpha_75_CB51B8 ...(truncated)"
    DeformRig_Bool="(upperarm_1_#4=True,elbow_3_#5=True,wrist_7_#6=True,groin_8_#7=True,glute_17_#8=True,knee_9_40C408324CA261 ...(truncated)"
    DeformRig_Debug=(upperarm_debug_18_#4=false,wrist_debug_20_#6=False,lat_debug_26_#9=False,groinGlute_debug_27_#7=False)
    DeformRig_Value="(upperarm_up_18_#10=(Rotation=(X=0,Y=0,Z=0,W=1),Translation=(X=-5,Y=0,Z=12),Scale3D=(X=0.7,Y=1,Z=1)),upperarm_fwd_fr_19_#11=(Rotation=(X=0,Y=0,Z=0,W=1),Translation=(X=0,Y=10,Z=0),Scale3D=(X=0.5,Y=1, ...(truncated)"
    ExecuteContext=@RigVMFunction_Sequence_1.C
   )
   Divide = &RigVMFunction_MathDoubleDiv(
    B=2
    A=&RigVMFunction_MathDoubleAdd(A=RigVMVariableNode() {variable=AlphaSPR_thigh_R}.Value, B=RigVMVariableNode() {variable=AlphaSPR_thigh_L}.Value).Result
   )
   > RigUnit_OffsetTransformForItem = &RigUnit_OffsetTransformForItem(
    Item=(Type=Bone,Name="extra_glute_scale")
    OffsetTransform=(Rotation=(X=0,Y=0,Z=0,W=1),Translation=(X=0,Y=0,Z=0),Scale3D=(X=1,Y=1.5,Z=1))
    bPropagateToChildren=True
    Weight=@Divide.Result
   )
   > &RigUnit_OffsetTransformForItem(
    Item=(Type=Bone,Name="extra_groin_side_l")
    OffsetTransform=(Rotation=(X=0,Y=0,Z=0,W=1),Translation=(X=5,Y=15,Z=0),Scale3D=(X=1,Y=1.5,Z=1))
    bPropagateToChildren=True
    Weight=RigVMVariableNode() {variable=AlphaSPR_thigh_side_L}.Value
   )
   > &RigUnit_OffsetTransformForItem(
    Item=(Type=Bone,Name="extra_groin_side_r")
    OffsetTransform=(Rotation=(X=0,Y=0,Z=0,W=1),Translation=(X=-5,Y=-15,Z=0),Scale3D=(X=1,Y=1.5,Z=1))
    bPropagateToChildren=True
    Weight=RigVMVariableNode() {variable=AlphaSPR_thigh_side_R}.Value
   )
   /Script/ControlRig.RigUnit_AimItem(
    DebugSettings=(bEnabled=False,Scale=10,WorldOffset=(Rotation=(X=0,Y=0,Z=0,W=1),Translation=(X=0,Y=0,Z=0),Scale3D=(X=1,Y=1,Z=1)))
    Item=(Type=Bone,Name="dyn_strap_fr_l_base")
    Primary=(Weight=1,Axis=(X=1,Y=0,Z=0),Target=(X=-26.0358,Y=8.0364,Z=0),Kind=Location,Space=(Type=Bone,Name="thigh_l"))
    Secondary=(Weight=1,Axis=(X=0,Y=1,Z=0),Target=(X=0,Y=10,Z=0),Kind=Location,Space=(Type=Bone,Name="spine_02"))
    Weight=1
    ExecutePin=@RigVMFunction_Sequence_1.D
   ) {is=struct RigVMUnitNode}
   RigVMFunction_MathBoolToFloat = &RigVMFunction_MathBoolToFloat(Value=RigVMVariableNode() {variable=LOD 2-3}.Value)
   > &RigUnit_AimItem(
    DebugSettings=(bEnabled=False,Scale=10,WorldOffset=(Rotation=(X=0,Y=0,Z=0,W=1),Translation=(X=0,Y=0,Z=0),Scale3D=(X=1,Y=1,Z=1)))
    Item=(Type=Bone,Name="dyn_strap_fr_r_base")
    Primary=(Weight=1,Axis=(X=1,Y=0,Z=0),Target=(X=26.0358,Y=-5,Z=0),Kind=Location,Space=(Type=Bone,Name="thigh_r"))
    Secondary=(Weight=1,Axis=(X=0,Y=-1,Z=0),Target=(X=0,Y=10,Z=0),Kind=Location,Space=(Type=Bone,Name="spine_02"))
    Weight=@RigVMFunction_MathBoolToFloat.Result
   )
   RigVMFunction_Sequence_4 = RigVMAggregateNode(ExecuteContext=@RigVMFunction_Sequence_1.F)
   head_vector_nodes = RigVMCollapseNode()
   RigVMVariableNode(Value=@head_vector_nodes.Result, ExecuteContext=@RigVMFunction_Sequence_4.B) {variable=head_vector}
   > RigVMVariableNode(Value=-90) {variable=rotate_collar}
   > RigVMFunction_Sequence_2 = RigVMAggregateNode()
   DotRot_CollarHead = RigVMFunctionReferenceNode(
    BoneIn=(Type=Bone,Name="extra_collar_01_l")
    s_max=1
    s_min=0.9
    ExecuteContext=@RigVMFunction_Sequence_2.A
    angle=RigVMVariableNode() {variable=rotate_collar}.Value
   )
   > DotRot_CollarHead_2 = RigVMFunctionReferenceNode(BoneIn=(Type=Bone,Name="extra_collar_02_l"), s_max=1, s_min=0.9, angle=RigVMVariableNode() {variable=rotate_collar}.Value)
   > DotRot_CollarHead_3 = RigVMFunctionReferenceNode(BoneIn=(Type=Bone,Name="extra_collar_03_l"), s_max=1, s_min=0.9, angle=RigVMVariableNode() {variable=rotate_collar}.Value)
   > DotRot_CollarHead_5 = RigVMFunctionReferenceNode(BoneIn=(Type=Bone,Name="extra_collar_04_l"), s_max=1, s_min=0.9, angle=RigVMVariableNode() {variable=rotate_collar}.Value)
   > DotRot_CollarHead_4 = RigVMFunctionReferenceNode(BoneIn=(Type=Bone,Name="extra_collar_05_l"), s_max=1, s_min=0.9, angle=RigVMVariableNode() {variable=rotate_collar}.Value)
   DotRot_CollarHead_1 = RigVMFunctionReferenceNode(
    BoneIn=(Type=Bone,Name="extra_collar_m_l")
    s_max=1
    s_min=0.9
    ExecuteContext=@RigVMFunction_Sequence_2.B
    angle=RigVMVariableNode() {variable=rotate_collar}.Value
   )
   DotRot_CollarHead_7 = RigVMFunctionReferenceNode(
    BoneIn=(Type=Bone,Name="extra_collar_01_r")
    s_max=1
    s_min=0.9
    ExecuteContext=@RigVMFunction_Sequence_2.C
    angle=RigVMVariableNode() {variable=rotate_collar}.Value
   )
   > DotRot_CollarHead_8 = RigVMFunctionReferenceNode(BoneIn=(Type=Bone,Name="extra_collar_02_r"), s_max=1, s_min=0.9, angle=RigVMVariableNode() {variable=rotate_collar}.Value)
   > DotRot_CollarHead_9 = RigVMFunctionReferenceNode(BoneIn=(Type=Bone,Name="extra_collar_03_r"), s_max=1, s_min=0.9, angle=RigVMVariableNode() {variable=rotate_collar}.Value)
   > DotRot_CollarHead_11 = RigVMFunctionReferenceNode(BoneIn=(Type=Bone,Name="extra_collar_04_r"), s_max=1, s_min=0.9, angle=RigVMVariableNode() {variable=rotate_collar}.Value)
   > DotRot_CollarHead_10 = RigVMFunctionReferenceNode(BoneIn=(Type=Bone,Name="extra_collar_05_r"), s_max=1, s_min=0.9, angle=RigVMVariableNode() {variable=rotate_collar}.Value)
   RigVMFunction_Sequence_5 = RigVMAggregateNode(ExecuteContext=@RigVMFunction_Sequence_4.C)
   n9 = &RigVMFunction_Sequence(ExecuteContext=@RigVMFunction_Sequence_5.B)
   n10 = &RigUnit_SphericalPoseReader(
    ActiveRegionScaleFactors=(PositiveWidth=1,NegativeWidth=0.1,PositiveHeight=0.5,NegativeHeight=1)
    ActiveRegionSize=0.3
    Debug=(bDrawDebug=false,bDraw2D=false,bDrawLocalAxes=false,DebugScale=30,DebugSegments=30,DebugThickness=0.2)
    DriverAxis=(X=-1,Y=0,Z=0)
    DriverItem=(Type=Bone,Name="clavicle_r")
    FalloffRegionScaleFactors=(PositiveWidth=1,NegativeWidth=0.5,PositiveHeight=1,NegativeHeight=1)
    FalloffSize=0.5
    OptionalParentItem=(Type=Bone,Name="None")
    RotationOffset=(X=0,Y=220,Z=180)
    ExecutePin=@n9.B
   )
   > &RigUnit_OffsetTransformForItem(
    Item=(Type=Bone,Name="extra_shoulder_buckle_r")
    OffsetTransform=(Rotation=(X=0,Y=0,Z=0,W=1),Translation=(X=0,Y=-4,Z=0),Scale3D=(X=1,Y=1,Z=1))
    bPropagateToChildren=True
    Weight=@n10.OutputParam
   )
   n11 = &RigUnit_SphericalPoseReader(
    ActiveRegionScaleFactors=(PositiveWidth=1,NegativeWidth=0.1,PositiveHeight=1,NegativeHeight=0.5)
    ActiveRegionSize=0.3
    Debug=(bDrawDebug=false,bDraw2D=false,bDrawLocalAxes=false,DebugScale=30,DebugSegments=30,DebugThickness=0.2)
    DriverAxis=(X=1,Y=0,Z=0)
    DriverItem=(Type=Bone,Name="clavicle_l")
    FalloffRegionScaleFactors=(PositiveWidth=1,NegativeWidth=0.5,PositiveHeight=1,NegativeHeight=0.5)
    FalloffSize=0.5
    OptionalParentItem=(Type=Bone,Name="None")
    RotationOffset=(X=0,Y=40,Z=180)
    ExecutePin=@n9.A
   )
   > &RigUnit_OffsetTransformForItem(
    Item=(Type=Bone,Name="extra_shoulder_buckle_l")
    OffsetTransform=(Rotation=(X=0,Y=0,Z=0,W=1),Translation=(X=0,Y=4,Z=0),Scale3D=(X=1,Y=1,Z=1))
    bPropagateToChildren=True
    Weight=@n11.OutputParam
   )
   n12 = &RigVMFunction_Sequence(ExecuteContext=@RigVMFunction_Sequence_5.C)
   n13 = &RigUnit_SphericalPoseReader(
    ActiveRegionScaleFactors=(PositiveWidth=1,NegativeWidth=1,PositiveHeight=1,NegativeHeight=1)
    ActiveRegionSize=0.1
    Debug=(bDrawDebug=true,bDraw2D=false,bDrawLocalAxes=false,DebugScale=30,DebugSegments=30,DebugThickness=0.2)
    DriverAxis=(X=1,Y=0,Z=0)
    DriverItem=(Type=Bone,Name="upperarm_l")
    FalloffRegionScaleFactors=(PositiveWidth=1,NegativeWidth=1,PositiveHeight=1,NegativeHeight=1)
    FalloffSize=0.3
    OptionalParentItem=(Type=Bone,Name="None")
    RotationOffset=(X=61,Y=0,Z=-28)
    ExecutePin=@n12.A
   )
   > &RigUnit_OffsetTransformForItem(
    Item=(Type=Bone,Name="extra_shoulder_buckle_l")
    OffsetTransform=(Rotation=(X=-0.0858,Y=0.0151,Z=-0.173,W=0.9811),Translation=(X=0,Y=-1.5,Z=0),Scale3D=(X=1,Y=1,Z=1))
    bPropagateToChildren=True
    Weight=@n13.OutputParam
   )
   > &RigUnit_OffsetTransformForItem(
    Item=(Type=Bone,Name="extra_collar_03_l")
    OffsetTransform=(Rotation=(X=0,Y=0,Z=-0.2588,W=0.9659),Translation=(X=0,Y=0,Z=0),Scale3D=(X=1,Y=1,Z=1))
    bPropagateToChildren=True
    Weight=@n13.OutputParam
   )
   n14 = &RigUnit_SphericalPoseReader(
    ActiveRegionScaleFactors=(PositiveWidth=1,NegativeWidth=1,PositiveHeight=1,NegativeHeight=1)
    ActiveRegionSize=0.1
    Debug=(bDrawDebug=true,bDraw2D=false,bDrawLocalAxes=false,DebugScale=30,DebugSegments=30,DebugThickness=0.2)
    DriverAxis=(X=-1,Y=0,Z=0)
    DriverItem=(Type=Bone,Name="upperarm_r")
    FalloffRegionScaleFactors=(PositiveWidth=1,NegativeWidth=1,PositiveHeight=1,NegativeHeight=1)
    FalloffSize=0.3
    OptionalParentItem=(Type=Bone,Name="None")
    RotationOffset=(X=255,Y=0,Z=33)
    ExecutePin=@n12.B
   )
   > &RigUnit_OffsetTransformForItem(
    Item=(Type=Bone,Name="extra_shoulder_buckle_r")
    OffsetTransform=(Rotation=(X=0,Y=0,Z=0,W=1),Translation=(X=0,Y=2,Z=0),Scale3D=(X=1,Y=1,Z=1))
    bPropagateToChildren=True
    Weight=@n14.OutputParam
   )
   n15 = &RigUnit_SphericalPoseReader(
    ActiveRegionScaleFactors=(PositiveWidth=0.3,NegativeWidth=0.3,PositiveHeight=0.3,NegativeHeight=0.3)
    ActiveRegionSize=0.3
    Debug=(bDrawDebug=false,bDraw2D=false,bDrawLocalAxes=false,DebugScale=30,DebugSegments=30,DebugThickness=0.2)
    DriverAxis=(X=-1,Y=0,Z=0)
    DriverItem=(Type=Bone,Name="upperarm_r")
    FalloffRegionScaleFactors=(PositiveWidth=0.1,NegativeWidth=1,PositiveHeight=1,NegativeHeight=0.5)
    FalloffSize=0.5
    OptionalParentItem=(Type=Bone,Name="None")
    RotationOffset=(X=-40,Y=220,Z=0)
    ExecutePin=@RigVMFunction_Sequence_5.D
   )
   RigVMFunction_ControlFlowBranch = &RigVMFunction_ControlFlowBranch(ExecuteContext=@n15.ExecutePin, Condition=RigVMVariableNode() {variable=crouch}.Value)
   n16 = &RigVMFunction_AlphaInterp(
    InterpSpeedDecreasing=10
    InterpSpeedIncreasing=10
    Scale=1
    bClampResult=False
    Value=&RigVMFunction_MathDoubleMul(A=@n15.OutputParam, B=RigVMVariableNode() {variable=targetAlpha}.Value).Result
   )
   &RigUnit_ModifyTransforms(
    ItemToModify="((Item=(Type=Bone,Name=\"extra_collar_01_r\"),Transform=(Rotation=(X=0,Y=0,Z=0,W=1),Translation=(X=0,Y=0,Z=0),Scale3D=(X=1,Y=1,Z=1))),(Item=(Type=Bone,Name=\"extra_collar_02_r\"),Transform=(Rotation=(X=0,Y=0,Z=-0.1736,W=0.9848),Translation=(X=0,Y=0,Z=0),Scale3 ...(truncated)"
    WeightMinimum=0
    Weight=@n16.Result
    ExecutePin=@RigVMFunction_ControlFlowBranch.False
   )
   &RigUnit_ModifyTransforms(
    ItemToModify="((Item=(Type=Bone,Name=\"extra_collar_01_r\"),Transform=(Rotation=(X=0,Y=0,Z=0,W=1),Translation=(X=0,Y=0,Z=0),Scale3D=(X=1,Y=1,Z=1))),(Item=(Type=Bone,Name=\"extra_collar_02_r\"),Transform=(Rotation=(X=0,Y=0,Z=-0.2588,W=0.9659),Translation=(X=0,Y=0,Z=0),Scale3 ...(truncated)"
    WeightMinimum=0
    Weight=@n16.Result
    ExecutePin=@RigVMFunction_ControlFlowBranch.True
   )
   RigVMVariableNode(
    Value=&RigVMFunction_MathBoolToFloat(Value=RigVMVariableNode() {variable=target}.Value).Result
    ExecuteContext=@RigVMFunction_Sequence_5.A
   ) {variable=targetAlpha}
   RigVMVariableNode(Value=1, ExecuteContext=@RigVMFunction_Sequence_4.A) {variable=contrs}
   > RigVMFunction_Sequence_6 = RigVMAggregateNode()
   n17 = RigVMVariableNode() {variable=contrs}
   &RigUnit_ParentConstraint(
    AdvancedSettings=(RotationOrderForFilter=XZY)
    Child=(Type=Bone,Name="extra_collar_01_l")
    Parents="((Item=(Type=Bone,Name=\"spine_03\"),Weight=0.0139),(Item=(Type=Bone,Name=\"neck_02\"),Weight=0.0222),(Item=(Type=Bone,Name=\"clavicle_r\"),Weight=0.0314),(Item=(Type=Bone,Name=\"clavicle_l\"),Weight=0.2051),(Item=(Type=Bone,Name=\"neck_01\"),Weight=0.2577),(Item=(T ...(truncated)"
    ExecutePin=@RigVMFunction_Sequence_6.A
    Weight=@n17.Value
   )
   > &RigUnit_ParentConstraint(
    AdvancedSettings=(RotationOrderForFilter=XZY)
    Child=(Type=Bone,Name="extra_collar_02_l")
    Parents="((Item=(Type=Bone,Name=\"deform_upperarm_bk_l\"),Weight=0.0044),(Item=(Type=Bone,Name=\"spine_03\"),Weight=0.0062),(Item=(Type=Bone,Name=\"deform_upperarm_up_l\"),Weight=0.0123),(Item=(Type=Bone,Name=\"neck_02\"),Weight=0.0135),(Item=(Type=Bone,Name=\"neck_01\"),Wei ...(truncated)"
    Weight=@n17.Value
   )
   > &RigUnit_ParentConstraint(
    AdvancedSettings=(RotationOrderForFilter=XZY)
    Child=(Type=Bone,Name="extra_collar_03_l")
    Parents="((Item=(Type=Bone,Name=\"neck_02\"),Weight=0.0058),(Item=(Type=Bone,Name=\"upperarm_twist_01_l\"),Weight=0.0188),(Item=(Type=Bone,Name=\"deform_pec_l\"),Weight=0.0263),(Item=(Type=Bone,Name=\"neck_01\"),Weight=0.1995),(Item=(Type=Bone,Name=\"spine_05\"),Weight=0.335 ...(truncated)"
    Weight=@n17.Value
   )
   > &RigUnit_ParentConstraint(
    AdvancedSettings=(RotationOrderForFilter=XZY)
    Child=(Type=Bone,Name="extra_collar_04_l")
    Parents="((Item=(Type=Bone,Name=\"upperarm_twist_01_l\"),Weight=0.0146),(Item=(Type=Bone,Name=\"spine_04\"),Weight=0.0165),(Item=(Type=Bone,Name=\"deform_pec_l\"),Weight=0.0723),(Item=(Type=Bone,Name=\"neck_01\"),Weight=0.1394),(Item=(Type=Bone,Name=\"spine_05\"),Weight=0.34 ...(truncated)"
    Weight=&RigVMFunction_MathDoubleMul(B=0.5, A=@n17.Value).Result
   )
   > &RigUnit_ParentConstraint(
    AdvancedSettings=(RotationOrderForFilter=XZY)
    Child=(Type=Bone,Name="extra_collar_05_l")
    Parents="((Item=(Type=Bone,Name=\"upperarm_twist_01_l\"),Weight=0.012),(Item=(Type=Bone,Name=\"spine_04\"),Weight=0.0355),(Item=(Type=Bone,Name=\"neck_01\"),Weight=0.065),(Item=(Type=Bone,Name=\"deform_pec_l\"),Weight=0.1195),(Item=(Type=Bone,Name=\"clavicle_l\"),Weight=0.34 ...(truncated)"
    Weight=&RigVMFunction_MathDoubleMul(B=0.5, A=@n17.Value).Result
   )
   &RigUnit_ParentConstraint(
    AdvancedSettings=(RotationOrderForFilter=XZY)
    Child=(Type=Bone,Name="extra_collar_m_l")
    Parents="((Item=(Type=Bone,Name=\"spine_04\"),Weight=0.0012),(Item=(Type=Bone,Name=\"head\"),Weight=0.0045),(Item=(Type=Bone,Name=\"spine_03\"),Weight=0.0201),(Item=(Type=Bone,Name=\"neck_02\"),Weight=0.0287),(Item=(Type=Bone,Name=\"clavicle_l\"),Weight=0.0782),(Item=(Type=B ...(truncated)"
    ExecutePin=@RigVMFunction_Sequence_6.B
    Weight=@n17.Value
   )
   &RigUnit_ParentConstraint(
    AdvancedSettings=(RotationOrderForFilter=XZY)
    Child=(Type=Bone,Name="extra_collar_01_r")
    Parents="((Item=(Type=Bone,Name=\"spine_03\"),Weight=0.0139),(Item=(Type=Bone,Name=\"neck_02\"),Weight=0.0222),(Item=(Type=Bone,Name=\"clavicle_l\"),Weight=0.0314),(Item=(Type=Bone,Name=\"clavicle_r\"),Weight=0.2051),(Item=(Type=Bone,Name=\"neck_01\"),Weight=0.2577),(Item=(T ...(truncated)"
    ExecutePin=@RigVMFunction_Sequence_6.C
    Weight=@n17.Value
   )
   > &RigUnit_ParentConstraint(
    AdvancedSettings=(RotationOrderForFilter=XZY)
    Child=(Type=Bone,Name="extra_collar_02_r")
    Parents="((Item=(Type=Bone,Name=\"deform_upperarm_bk_r\"),Weight=0.0044),(Item=(Type=Bone,Name=\"spine_03\"),Weight=0.0062),(Item=(Type=Bone,Name=\"deform_upperarm_up_r\"),Weight=0.0123),(Item=(Type=Bone,Name=\"neck_02\"),Weight=0.0135),(Item=(Type=Bone,Name=\"neck_01\"),Wei ...(truncated)"
    Weight=@n17.Value
   )
   > &RigUnit_ParentConstraint(
    AdvancedSettings=(RotationOrderForFilter=XZY)
    Child=(Type=Bone,Name="extra_collar_03_r")
    Parents="((Item=(Type=Bone,Name=\"neck_02\"),Weight=0.0058),(Item=(Type=Bone,Name=\"upperarm_twist_01_r\"),Weight=0.0188),(Item=(Type=Bone,Name=\"deform_pec_r\"),Weight=0.0263),(Item=(Type=Bone,Name=\"neck_01\"),Weight=0.1995),(Item=(Type=Bone,Name=\"spine_05\"),Weight=0.335 ...(truncated)"
    Weight=@n17.Value
   )
   > &RigUnit_ParentConstraint(
    AdvancedSettings=(RotationOrderForFilter=XZY)
    Child=(Type=Bone,Name="extra_collar_04_r")
    Parents="((Item=(Type=Bone,Name=\"upperarm_twist_01_r\"),Weight=0.0146),(Item=(Type=Bone,Name=\"spine_04\"),Weight=0.0165),(Item=(Type=Bone,Name=\"deform_pec_r\"),Weight=0.0723),(Item=(Type=Bone,Name=\"neck_01\"),Weight=0.1394),(Item=(Type=Bone,Name=\"spine_05\"),Weight=0.34 ...(truncated)"
    Weight=&RigVMFunction_MathDoubleMul(B=0.5, A=@n17.Value).Result
   )
   > &RigUnit_ParentConstraint(
    AdvancedSettings=(RotationOrderForFilter=XZY)
    Child=(Type=Bone,Name="extra_collar_05_r")
    Parents="((Item=(Type=Bone,Name=\"upperarm_twist_01_r\"),Weight=0.012),(Item=(Type=Bone,Name=\"spine_04\"),Weight=0.0355),(Item=(Type=Bone,Name=\"neck_01\"),Weight=0.065),(Item=(Type=Bone,Name=\"deform_pec_r\"),Weight=0.1195),(Item=(Type=Bone,Name=\"clavicle_r\"),Weight=0.34 ...(truncated)"
    Weight=&RigVMFunction_MathDoubleMul(B=0.5, A=@n17.Value).Result
   )
   &RigUnit_ParentConstraint(
    AdvancedSettings=(RotationOrderForFilter=XZY)
    Child=(Type=Bone,Name="extra_shoulder_buckle_l")
    Parents="((Item=(Type=Bone,Name=\"deform_upperarm_bk_l\"),Weight=0.0122),(Item=(Type=Bone,Name=\"deform_pec_l\"),Weight=0.018),(Item=(Type=Bone,Name=\"spine_04\"),Weight=0.021),(Item=(Type=Bone,Name=\"deform_upperarm_fr_l\"),Weight=0.0647),(Item=(Type=Bone,Name=\"deform_upp ...(truncated)"
    Weight=1
    ExecutePin=@RigVMFunction_Sequence_1.E
   )
   > &RigUnit_ParentConstraint(
    AdvancedSettings=(RotationOrderForFilter=XZY)
    Child=(Type=Bone,Name="extra_shoulder_buckle_r")
    Parents="((Item=(Type=Bone,Name=\"deform_upperarm_bk_r\"),Weight=0.0122),(Item=(Type=Bone,Name=\"deform_pec_r\"),Weight=0.018),(Item=(Type=Bone,Name=\"spine_04\"),Weight=0.021),(Item=(Type=Bone,Name=\"deform_upperarm_fr_r\"),Weight=0.0647),(Item=(Type=Bone,Name=\"deform_upp ...(truncated)"
    Weight=1
   )
   n18 = &RigVMFunction_AlphaInterp(
    InterpSpeedDecreasing=10
    InterpSpeedIncreasing=10
    Scale=1
    bClampResult=False
    Value=&RigVMFunction_MathDoubleMul(A=@n11.OutputParam, B=RigVMVariableNode() {variable=targetAlpha}.Value).Result
   )
   &RigUnit_ModifyTransforms(
    ExecutePin=()
    ItemToModify="((Item=(Type=Bone,Name=\"extra_collar_06_l\"),Transform=(Rotation=(X=0,Y=0,Z=0,W=1),Translation=(X=0,Y=-0.5,Z=0),Scale3D=(X=1,Y=1,Z=1))),(Item=(Type=Bone,Name=\"extra_collar_05_l\"),Transform=(Rotation=(X=0,Y=0,Z=0,W=1),Translation=(X=0,Y=-1,Z=0),Scale3D=(X=1, ...(truncated)"
    WeightMinimum=0
    Weight=@n18.Result
   )
   n19 = &RigVMFunction_AlphaInterp(
    InterpSpeedDecreasing=10
    InterpSpeedIncreasing=10
    Scale=1
    bClampResult=False
    Value=&RigVMFunction_MathDoubleMul(A=@n10.OutputParam, B=RigVMVariableNode() {variable=targetAlpha}.Value).Result
   )
   &RigUnit_ModifyTransforms(
    ExecutePin=()
    ItemToModify="((Item=(Type=Bone,Name=\"extra_collar_01_r\"),Transform=(Rotation=(X=0,Y=0,Z=0,W=1),Translation=(X=0,Y=0,Z=0),Scale3D=(X=1,Y=1,Z=1))),(Item=(Type=Bone,Name=\"extra_collar_02_r\"),Transform=(Rotation=(X=0,Y=0,Z=0.342,W=0.9397),Translation=(X=0,Y=0,Z=0),Scale3D= ...(truncated)"
    WeightMinimum=0
    Weight=@n19.Result
   )
   &RigUnit_ParentConstraint(
    AdvancedSettings=(InterpolationType=Average,RotationOrderForFilter=XZY)
    Child=(Type=Bone,Name="extra_shoulder_buckle_l")
    ExecutePin=()
    Filter=(TranslationFilter=(bX=False,bY=False,bZ=False),RotationFilter=(bX=True,bY=True,bZ=True),ScaleFilter=(bX=True,bY=True,bZ=True))
    Parents=((Item=(Type=Bone,Name="clavicle_l"),Weight=1))
    Weight=0.5
   )
   &RigUnit_ParentConstraint(
    AdvancedSettings=(RotationOrderForFilter=XZY)
    Child=(Type=Bone,Name="extra_shoulder_buckle_l")
    ExecutePin=()
    Parents="((Item=(Type=Bone,Name=\"deform_upperarm_bk_l\"),Weight=0.0122),(Item=(Type=Bone,Name=\"deform_pec_l\"),Weight=0.018),(Item=(Type=Bone,Name=\"spine_04\"),Weight=0.021),(Item=(Type=Bone,Name=\"deform_upperarm_fr_l\"),Weight=0.0647),(Item=(Type=Bone,Name=\"deform_upp ...(truncated)"
    Weight=1
   )
   &RigUnit_SphericalPoseReader(
    ActiveRegionScaleFactors=(PositiveWidth=3.5,NegativeWidth=0.5,PositiveHeight=3,NegativeHeight=3)
    ActiveRegionSize=0.1
    Debug=(bDrawDebug=true,bDraw2D=false,DebugScale=25,DebugSegments=20,DebugThickness=0.25)
    DriverAxis=(X=1,Y=0,Z=0)
    DriverItem=(Type=Bone,Name="thigh_r")
    ExecutePin=()
    FalloffRegionScaleFactors=(PositiveWidth=1,NegativeWidth=1,PositiveHeight=1,NegativeHeight=1)
    FalloffSize=0.45
    FlipHeightScaling=false
    FlipWidthScaling=true
    OptionalParentItem=(Type=Bone)
    RotationOffset=(X=0,Y=0,Z=0)
   )
  ]}
]
hierarchy [
 root {loc=<0 0 0>, rot=<0 0 0 1>}
 pelvis {parent=root, loc=<0 -0.5301 98.0571>, rot=<0.0127 0.707 -0.0127 -0.707>}
 spine_01 {parent=pelvis, loc=<6.1208 -0.9328 0>, rot=<0 0 0.034 -0.9994>}
 spine_02 {parent=spine_01, loc=<6.5997 -0.2095 0>, rot=<0 0 0.0238 -0.9997>}
 spine_03 {parent=spine_02, loc=<6.4914 -0.1237 0>, rot=<0 0 0.0149 -0.9999>}
 spine_04 {parent=spine_03, loc=<6.9598 0.0794 0>, rot=<0 0 -0.0132 -0.9999>}
 spine_05 {parent=spine_04, loc=<6.9987 0 0>, rot=<0 0 -0.0015 -1>}
 neck_01 {parent=spine_05, loc=<11.3138 0.1529 0>, rot=<0 0 -0.2003 -0.9797>}
 neck_02 {parent=neck_01, loc=<4.5627 0 0>, rot=<0 0 0.0504 -0.9987>}
 head {parent=neck_02, loc=<5.3914 0 0>, rot=<0 0 0.115 -0.9934>}
 head_end {parent=head, loc=<10 0 0>, rot=<0 0 0 1>}
 clavicle_l {parent=spine_05, loc=<5.4831 -0.1053 -3.2946>, rot=<0.0465 -0.7289 -0.009 -0.6829>}
 upperarm_l {parent=clavicle_l, loc=<9.8658 -1.2136 -0.4105>, rot=<0.0164 -0.3245 -0.0269 -0.9454>}
 lowerarm_l {parent=upperarm_l, loc=<26.465 -0.8345 0.259>, rot=<0.0036 0.0161 -0.2298 -0.9731>}
 hand_l {parent=lowerarm_l, loc=<22.7355 -0.3203 -0.6105>, rot=<-0.6784 -0.0223 0.0735 -0.7307>}
 thumb_01_l {parent=hand_l, loc=<1.3118 -0.3191 -2.459>, rot=<0.6546 -0.4042 0.0664 -0.6354>}
 thumb_02_l {parent=thumb_01_l, loc=<5.9336 -0.0448 0.0401>, rot=<-0.0037 0.0213 0.298 -0.9543>}
 thumb_03_l {parent=thumb_02_l, loc=<2.2223 0.3682 -0.0907>, rot=<0.0118 -0.013 0.0066 -0.9998>}
 index_metacarpal_l {parent=hand_l, loc=<1.6535 0.0183 -2.96>, rot=<0.0158 -0.0124 -0.0173 -0.9996>}
 index_01_l {parent=index_metacarpal_l, loc=<8.3008 0.0357 -0.0524>, rot=<-0.0044 0.0485 0.0924 -0.9945>}
 index_02_l {parent=index_01_l, loc=<3.3587 -0.3004 0.0452>, rot=<-0.002 0.0027 0.3225 -0.9466>}
 index_03_l {parent=index_02_l, loc=<2.5578 -0.123 -0.0252>, rot=<-0.002 0.0067 0.2558 -0.9667>}
 middle_metacarpal_l {parent=hand_l, loc=<2.1856 -0.1725 -0.5984>, rot=<0.006 0.0168 -0.0329 -0.9993>}
 middle_01_l {parent=middle_metacarpal_l, loc=<7.3275 -0.2447 0.0129>, rot=<-0.0208 0.0042 0.0858 -0.9961>}
 middle_02_l {parent=middle_01_l, loc=<4.0123 0.1401 0.1175>, rot=<-0.0218 0.0258 0.3512 -0.9357>}
 middle_03_l {parent=middle_02_l, loc=<2.6077 0.0534 0.1373>, rot=<0.0082 0.0188 0.1708 -0.9851>}
 ring_metacarpal_l {parent=hand_l, loc=<2.2074 -0.3625 1.3638>, rot=<-0.0284 0.0722 -0.0458 -0.9959>}
 ring_01_l {parent=ring_metacarpal_l, loc=<7.0295 -0.6197 -0.1446>, rot=<-0.0489 -0.0252 0.1892 -0.9804>}
 ring_02_l {parent=ring_01_l, loc=<3.6174 0.4647 -0.1027>, rot=<0 0 0.2806 -0.9598>}
 ring_03_l {parent=ring_02_l, loc=<2.4152 0.114 0.0199>, rot=<0 0 0.1189 -0.9929>}
 pinky_metacarpal_l {parent=hand_l, loc=<1.7616 -0.9753 3.3814>, rot=<-0.0397 0.0814 -0.069 -0.9935>}
 pinky_01_l {parent=pinky_metacarpal_l, loc=<7.1709 -0.4831 -0.1094>, rot=<-0.0654 -0.0035 0.2242 -0.9723>}
 pinky_02_l {parent=pinky_01_l, loc=<2.8274 0.0893 -0.1043>, rot=<0.0037 0.0011 0.2854 -0.9584>}
 pinky_03_l {parent=pinky_02_l, loc=<2.2046 0.0307 -0.0078>, rot=<-0.0002 0.0098 0.1445 -0.9895>}
 deform_wrist_dn_base_l {parent=hand_l, loc=<0 -1.5326 0>, rot=<0 0 0 1>}
 deform_wrist_dn_l {parent=deform_wrist_dn_base_l, loc=<0 0 0>, rot=<0 0 0 1>}
 deform_wrist_up_base_l {parent=hand_l, loc=<0 2.3009 0>, rot=<0 0 0 1>}
 deform_wrist_up_l {parent=deform_wrist_up_base_l, loc=<0 0 0>, rot=<0 0 0 1>}
 lowerarm_twist_02_l {parent=lowerarm_l, loc=<10.6707 0 0>, rot=<0 0 0 1>}
 lowerarm_twist_01_l {parent=lowerarm_l, loc=<18.316 0 0>, rot=<0 0 0 1>}
 deform_elbow_in_base_l {parent=lowerarm_l, loc=<0.1834 3.6062 -0.2572>, rot=<0 0 0 1>}
 deform_elbow_in_l {parent=deform_elbow_in_base_l, loc=<0 0 0>, rot=<0 0 0 1>}
 deform_elbow_out_base_l {parent=lowerarm_l, loc=<0 -4 0>, rot=<0 0 0 1>}
 deform_elbow_out_l {parent=deform_elbow_out_base_l, loc=<0 0 0>, rot=<0 0 0 1>}
 upperarm_twist_01_l {parent=upperarm_l, loc=<7.2135 -0.1134 0>, rot=<0 0 0 1>}
 deform_upperarm_bk_base_l {parent=upperarm_twist_01_l, loc=<-6.4053 -4.3746 0.314>, rot=<0 0 0 1>}
 deform_upperarm_bk_l {parent=deform_upperarm_bk_base_l, loc=<0 0 0>, rot=<0 0 0 1>}
 deform_upperarm_fr_base_l {parent=upperarm_twist_01_l, loc=<-5.6312 5.092 0.3354>, rot=<0 0 0 1>}
 deform_upperarm_fr_l {parent=deform_upperarm_fr_base_l, loc=<0 0 0>, rot=<0 0 0 1>}
 deform_upperarm_up_base_l {parent=upperarm_twist_01_l, loc=<-8.891 0.4825 3.9517>, rot=<0 0 0 1>}
 deform_upperarm_up_l {parent=deform_upperarm_up_base_l, loc=<0 0 0>, rot=<0 0 0 1>}
 upperarm_twist_02_l {parent=upperarm_l, loc=<19.7998 -0.1134 0>, rot=<0 0 0 1>}
 extra_shoulder_buckle_l {parent=clavicle_l, loc=<7.8499 3.0728 5.8889>, rot=<0.6734 0.5493 0.271 -0.414>}
 clavicle_r {parent=spine_05, loc=<5.4831 -0.1053 3.2946>, rot=<0.7289 0.0465 -0.6829 0.009>}
 upperarm_r {parent=clavicle_r, loc=<-9.8658 1.2136 0.4105>, rot=<0.0164 -0.3245 -0.0269 -0.9454>}
 lowerarm_r {parent=upperarm_r, loc=<-26.465 0.8345 -0.259>, rot=<0.0036 0.0161 -0.2298 -0.9731>}
 hand_r {parent=lowerarm_r, loc=<-22.7355 0.3203 0.6106>, rot=<-0.6784 -0.0223 0.0735 -0.7307>}
 thumb_01_r {parent=hand_r, loc=<-1.3118 0.3191 2.459>, rot=<0.6546 -0.4042 0.0664 -0.6354>}
 thumb_02_r {parent=thumb_01_r, loc=<-5.9336 0.0448 -0.0401>, rot=<-0.0037 0.0213 0.298 -0.9543>}
 thumb_03_r {parent=thumb_02_r, loc=<-2.2223 -0.3682 0.0907>, rot=<0.0118 -0.013 0.0066 -0.9998>}
 index_metacarpal_r {parent=hand_r, loc=<-1.6535 -0.0183 2.96>, rot=<0.0158 -0.0124 -0.0173 -0.9996>}
 index_01_r {parent=index_metacarpal_r, loc=<-8.3008 -0.0357 0.0524>, rot=<-0.0044 0.0485 0.0924 -0.9945>}
 index_02_r {parent=index_01_r, loc=<-3.3587 0.3004 -0.0452>, rot=<-0.002 0.0027 0.3225 -0.9466>}
 index_03_r {parent=index_02_r, loc=<-2.5578 0.123 0.0252>, rot=<-0.002 0.0067 0.2558 -0.9667>}
 middle_metacarpal_r {parent=hand_r, loc=<-2.1856 0.1725 0.5984>, rot=<0.006 0.0168 -0.0329 -0.9993>}
 middle_01_r {parent=middle_metacarpal_r, loc=<-7.3275 0.2447 -0.0129>, rot=<-0.0208 0.0042 0.0858 -0.9961>}
 middle_02_r {parent=middle_01_r, loc=<-4.0123 -0.1401 -0.1175>, rot=<-0.0218 0.0258 0.3512 -0.9357>}
 middle_03_r {parent=middle_02_r, loc=<-2.6077 -0.0534 -0.1373>, rot=<0.0082 0.0188 0.1708 -0.9851>}
 ring_metacarpal_r {parent=hand_r, loc=<-2.2074 0.3625 -1.3638>, rot=<-0.0284 0.0722 -0.0458 -0.9959>}
 ring_01_r {parent=ring_metacarpal_r, loc=<-7.0295 0.6197 0.1446>, rot=<-0.0489 -0.0252 0.1892 -0.9804>}
 ring_02_r {parent=ring_01_r, loc=<-3.6174 -0.4647 0.1027>, rot=<0 0 0.2806 -0.9598>}
 ring_03_r {parent=ring_02_r, loc=<-2.4152 -0.114 -0.0199>, rot=<0 0 0.1189 -0.9929>}
 pinky_metacarpal_r {parent=hand_r, loc=<-1.7616 0.9753 -3.3814>, rot=<-0.0397 0.0814 -0.069 -0.9935>}
 pinky_01_r {parent=pinky_metacarpal_r, loc=<-7.1709 0.4831 0.1094>, rot=<-0.0654 -0.0035 0.2242 -0.9723>}
 pinky_02_r {parent=pinky_01_r, loc=<-2.8274 -0.0893 0.1043>, rot=<0.0037 0.0011 0.2854 -0.9584>}
 pinky_03_r {parent=pinky_02_r, loc=<-2.2046 -0.0307 0.0078>, rot=<-0.0002 0.0098 0.1445 -0.9895>}
 deform_wrist_dn_base_r {parent=hand_r, loc=<-0.0001 1.5328 0>, rot=<0 0 0 1>}
 deform_wrist_dn_r {parent=deform_wrist_dn_base_r, loc=<0 0 0>, rot=<0 0 0 1>}
 deform_wrist_up_base_r {parent=hand_r, loc=<-0.0002 -2.3007 -0.0001>, rot=<0 0 0 1>}
 deform_wrist_up_r {parent=deform_wrist_up_base_r, loc=<0 0 0>, rot=<0 0 0 1>}
 lowerarm_twist_02_r {parent=lowerarm_r, loc=<-10.6707 0 0>, rot=<0 0 0 -1>}
 lowerarm_twist_01_r {parent=lowerarm_r, loc=<-18.316 0 0>, rot=<0 0 0 -1>}
 deform_elbow_in_base_r {parent=lowerarm_r, loc=<-0.1834 -3.6062 0.2572>, rot=<0 0 0 -1>}
 deform_elbow_in_r {parent=deform_elbow_in_base_r, loc=<0 0 0>, rot=<0 0 0 1>}
 deform_elbow_out_base_r {parent=lowerarm_r, loc=<0 4 -0.0001>, rot=<0 0 0 -1>}
 deform_elbow_out_r {parent=deform_elbow_out_base_r, loc=<0 0 0>, rot=<0 0 0 1>}
 upperarm_twist_01_r {parent=upperarm_r, loc=<-7.2135 0.1134 0>, rot=<0 0 0 -1>}
 deform_upperarm_bk_base_r {parent=upperarm_twist_01_r, loc=<6.4053 4.3746 -0.314>, rot=<0 0 0 1>}
 deform_upperarm_bk_r {parent=deform_upperarm_bk_base_r, loc=<0 0 0>, rot=<0 0 0 1>}
 deform_upperarm_fr_base_r {parent=upperarm_twist_01_r, loc=<5.6312 -5.092 -0.3354>, rot=<0 0 0 1>}
 deform_upperarm_fr_r {parent=deform_upperarm_fr_base_r, loc=<0 0 0>, rot=<0 0 0 1>}
 deform_upperarm_up_base_r {parent=upperarm_twist_01_r, loc=<8.891 -0.4825 -3.9517>, rot=<0 0 0 1>}
 deform_upperarm_up_r {parent=deform_upperarm_up_base_r, loc=<0 0 0>, rot=<0 0 0 1>}
 upperarm_twist_02_r {parent=upperarm_r, loc=<-19.7998 0.1134 0>, rot=<0 0 0 -1>}
 extra_shoulder_buckle_r {parent=clavicle_r, loc=<-7.8499 -3.0728 -5.8889>, rot=<-0.271 -0.414 0.6734 -0.5493>}
 attach_backpack {parent=spine_05, loc=<-0.123 -13.473 0>, rot=<-0.0093 -0.707 0.0093 -0.707>}
 Backpack_BR {parent=attach_backpack, type=Null, loc=<0 -2.4723 0.0015>, rot=<-0.0301 0 0 0.9995>}
 attach_cape {parent=spine_05, loc=<9.017 -11.714 0>, rot=<0.0616 -0.7044 -0.0616 -0.7044>}
 cape {parent=attach_cape, type=Null, loc=<0 -1.533 0.2712>, rot=<0.1382 0 0 0.9904>}
 dyn_strap_bk_r_base {parent=spine_05, loc=<-1.8023 -8.9836 8.6419>, rot=<0.0049 -0.4983 0.867 -0.0068>}
 dyn_strap_bk_r_01 {parent=dyn_strap_bk_r_base, loc=<0.1 0 0>, rot=<0 0 0 1>}
 dyn_strap_bk_r_02 {parent=dyn_strap_bk_r_01, loc=<5.3821 0 0>, rot=<-0.0093 0.0288 -0.019 -0.9994>}
 dyn_strap_bk_l_base {parent=spine_05, loc=<-1.8019 -8.9835 -8.6419>, rot=<0.0068 -0.867 0.4983 -0.0049>}
 dyn_strap_bk_l_01 {parent=dyn_strap_bk_l_base, loc=<0.1002 0 0.0001>, rot=<0 0 0 1>}
 dyn_strap_bk_l_02 {parent=dyn_strap_bk_l_01, loc=<5.3819 0 -0.0001>, rot=<0.0093 0.0288 0.019 -0.9994>}
 dyn_strap_bk_c_base {parent=spine_05, loc=<10.9823 -11.0102 0>, rot=<0.4211 -0.568 0.568 -0.4211>}
 dyn_strap_bk_c_01 {parent=dyn_strap_bk_c_base, loc=<0.1002 0 0.0001>, rot=<0 0 0 1>}
 extra_collar_05_r {parent=spine_05, loc=<9.2055 5.2374 6.8197>, rot=<0.7134 0.111 -0.1756 -0.6692>}
 extra_collar_04_r {parent=spine_05, loc=<10.6399 2.5273 7.1491>, rot=<0.7039 0.1072 -0.1958 -0.6744>}
 extra_collar_03_r {parent=spine_05, loc=<11.7398 -0.1762 6.7328>, rot=<0.6164 0.0388 -0.1395 -0.774>}
 extra_collar_02_r {parent=spine_05, loc=<12.2111 -2.9113 5.4265>, rot=<0.4608 0.0187 -0.1296 -0.8778>}
 extra_collar_01_r {parent=spine_05, loc=<12.6014 -4.9759 3.2545>, rot=<0.2958 0.0226 -0.1352 -0.9454>}
 extra_collar_m_l {parent=spine_05, loc=<12.4604 -5.8883 -0.0027>, rot=<-0.0067 -0.0054 -0.1306 -0.9914>}
 extra_collar_01_l {parent=spine_05, loc=<12.6013 -4.9759 -3.2545>, rot=<-0.2958 -0.0226 -0.1352 -0.9454>}
 extra_collar_02_l {parent=spine_05, loc=<12.2116 -2.9113 -5.4265>, rot=<-0.4608 -0.0187 -0.1296 -0.8778>}
 extra_collar_03_l {parent=spine_05, loc=<11.7399 -0.1762 -6.7328>, rot=<-0.6164 -0.0388 -0.1395 -0.774>}
 extra_collar_04_l {parent=spine_05, loc=<10.6397 2.5273 -7.1491>, rot=<-0.7039 -0.1072 -0.1958 -0.6744>}
 extra_collar_05_l {parent=spine_05, loc=<9.2059 5.2374 -6.8197>, rot=<-0.7134 -0.111 -0.1756 -0.6692>}
 deform_pec_aim_base_l {parent=spine_04, loc=<3.2456 9.3309 -1.2237>, rot=<0.0603 -0.4033 0.3318 -0.8507>}
 deform_pec_aim_l {parent=deform_pec_aim_base_l, loc=<0 0 0>, rot=<0 0 0 1>}
 deform_pec_base_l {parent=deform_pec_aim_l, loc=<9.3665 0 0>, rot=<0 0 0 1>}
 deform_pec_l {parent=deform_pec_base_l, loc=<0 0 0>, rot=<0 0 0 1>}
 deform_pec_aim_base_r {parent=spine_04, loc=<3.2452 9.3309 1.2237>, rot=<-0.4033 -0.0603 0.8507 0.3318>}
 deform_pec_aim_r {parent=deform_pec_aim_base_r, loc=<0 0 0>, rot=<0 0 0 1>}
 deform_pec_base_r {parent=deform_pec_aim_r, loc=<-9.3667 -0.0003 -0.0004>, rot=<0 0 0 1>}
 deform_pec_r {parent=deform_pec_base_r, loc=<0 0 0>, rot=<0 0 0 1>}
 dyn_strap_fr_l_base {parent=spine_02, loc=<0.9674 6.6811 -6.9006>, rot=<-0.0338 -0.9328 -0.2975 -0.2003>}
 dyn_strap_fr_l_01 {parent=dyn_strap_fr_l_base, loc=<0.1 0 0>, rot=<0 0 0 1>}
 dyn_strap_fr_l_02 {parent=dyn_strap_fr_l_01, loc=<4.6358 -0.0155 -0.1128>, rot=<-0.0122 0.0986 0.0004 -0.9951>}
 dyn_strap_fr_l_03 {parent=dyn_strap_fr_l_02, loc=<4.7955 0 0>, rot=<0.076 -0.0108 -0.0009 -0.997>}
 dyn_strap_fr_l_extra {parent=dyn_strap_fr_l_02, loc=<4.7 0 0>, rot=<0.076 -0.0108 -0.0009 -0.997>}
 dyn_strap_fr_r_base {parent=spine_02, loc=<0.9678 6.6811 6.9006>, rot=<0.2003 0.2975 0.9328 0.0338>}
 dyn_strap_fr_r_01 {parent=dyn_strap_fr_r_base, loc=<0.1 0 0>, rot=<0 0 0 1>}
 dyn_strap_fr_r_02 {parent=dyn_strap_fr_r_01, loc=<4.6361 0.0155 -0.1129>, rot=<0.0122 0.0986 -0.0004 -0.9951>}
 dyn_strap_fr_r_03 {parent=dyn_strap_fr_r_02, loc=<4.7953 0 0.0001>, rot=<-0.076 -0.0108 0.0009 -0.997>}
 thigh_l {parent=pelvis, loc=<-1.0855 -0.2109 -8.7053>, rot=<0.0009 0.0355 0.0155 -0.9993>}
 calf_l {parent=thigh_l, loc=<-42.5284 -0.0019 -0.7297>, rot=<0.0006 -0.0161 -0.0391 -0.9991>}
 foot_l {parent=calf_l, loc=<-46.7401 -2.5014 -0.592>, rot=<0.0427 -0.0238 0.0353 -0.9982>}
 ball_l {parent=foot_l, loc=<-3.5658 10.8066 -0.1544>, rot=<0.0001 -0.0001 -0.7063 -0.7079>}
 calf_twist_02_l {parent=calf_l, loc=<-19.1546 -1.7943 -0.2638>, rot=<0.0237 0.0095 -0.0035 -0.9997>}
 calf_twist_01_l {parent=calf_l, loc=<-35.8483 -1.9197 -0.5794>, rot=<0.0237 0.0095 -0.0035 -0.9997>}
 deform_knee_in_base_l {parent=calf_l, loc=<0.8031 -5.0231 0.2456>, rot=<0 0 0 -1>}
 deform_knee_in_l {parent=deform_knee_in_base_l, loc=<0 0 0>, rot=<0 0 0 1>}
 deform_knee_out_base_l {parent=calf_l, loc=<0 4 0>, rot=<0 0 0 -1>}
 deform_knee_out_l {parent=deform_knee_out_base_l, loc=<0 0 0>, rot=<0 0 0 1>}
 thigh_twist_01_l {parent=thigh_l, loc=<-27.952 0 0>, rot=<0.0013 0 0 -1>}
 deform_groin_base_l {parent=thigh_l, loc=<0.4801 5.6776 -0.2486>, rot=<0 0 0 -1>}
 deform_groin_l {parent=deform_groin_base_l, loc=<0 0 0>, rot=<0 0 0 1>}
 extra_groin_side_base_l {parent=thigh_l, loc=<-3.4274 0.249 -2.0689>, rot=<0.7071 0 0 -0.7071>}
 extra_groin_side_l {parent=extra_groin_side_base_l, loc=<0 0 0>, rot=<0 0 0 1>}
 thigh_r {parent=pelvis, loc=<-1.0855 -0.2109 8.7053>, rot=<0.0355 -0.0009 0.9993 0.0155>}
 calf_r {parent=thigh_r, loc=<42.5284 0.0019 0.7297>, rot=<0.0006 -0.0161 -0.0391 -0.9991>}
 foot_r {parent=calf_r, loc=<46.7401 2.5014 0.592>, rot=<0.0427 -0.0238 0.0353 -0.9982>}
 ball_r {parent=foot_r, loc=<3.5658 -10.8066 0.1544>, rot=<0.0001 -0.0001 -0.7063 -0.7079>}
 calf_twist_02_r {parent=calf_r, loc=<19.1546 1.7943 0.2638>, rot=<0.0237 0.0095 -0.0035 -0.9997>}
 calf_twist_01_r {parent=calf_r, loc=<35.8483 1.9197 0.5794>, rot=<0.0237 0.0095 -0.0035 -0.9997>}
 deform_knee_in_base_r {parent=calf_r, loc=<-0.8031 5.0231 -0.2457>, rot=<0 0 0 -1>}
 deform_knee_in_r {parent=deform_knee_in_base_r, loc=<0 0 0>, rot=<0 0 0 1>}
 deform_knee_out_base_r {parent=calf_r, loc=<0 -4 0>, rot=<0 0 0 -1>}
 deform_knee_out_r {parent=deform_knee_out_base_r, loc=<0 0 0>, rot=<0 0 0 1>}
 thigh_twist_01_r {parent=thigh_r, loc=<27.952 0 0>, rot=<0.0013 0 0 -1>}
 deform_groin_base_r {parent=thigh_r, loc=<-0.4801 -5.6776 0.2486>, rot=<0 0 0 -1>}
 deform_groin_r {parent=deform_groin_base_r, loc=<0 0 0>, rot=<0 0 0 1>}
 extra_groin_side_base_r {parent=thigh_r, loc=<3.4273 -0.249 2.0689>, rot=<0.7071 0 0 -0.7071>}
 extra_groin_side_r {parent=extra_groin_side_base_r, loc=<0 0 0>, rot=<0 0 0 1>}
 deform_glute_base_l {parent=pelvis, loc=<-7.3466 -9.0213 -9.1245>, rot=<0.0172 0.031 -0.4477 -0.8935>}
 deform_glute_l {parent=deform_glute_base_l, loc=<0 0 0>, rot=<0 0 0 1>}
 deform_glute_base_r {parent=pelvis, loc=<-7.3464 -9.0213 9.1245>, rot=<0.031 -0.0172 0.8935 -0.4477>}
 deform_glute_r {parent=deform_glute_base_r, loc=<0 0 0>, rot=<0 0 0 1>}
 extra_glute_scale {parent=pelvis, loc=<-6.7913 -2.5021 0>, rot=<0 0 0 1>}
]
variables [
 AlphaSPR_thigh_L {type=float, default=0}
 AlphaSPR_thigh_R {type=float, default=0}
 AlphaSPR_thigh_side_L {type=float, default=0}
 AlphaSPR_thigh_side_R {type=float, default=0}
 LOD 2-3 {type=bool, default=False}
 contrs {type=double, default=0}
 crouch {type=bool, default=False}
 head_vector {type=FVector, default=(X=0,Y=0,Z=0)}
 mask {type=bool, default=False}
 rotate_collar {type=double, default=0}
 target {type=bool, default=False}
 targetAlpha {type=double, default=0}
]
