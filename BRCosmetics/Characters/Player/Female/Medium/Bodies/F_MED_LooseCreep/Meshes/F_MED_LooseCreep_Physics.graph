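# Physics asset for the LooseCreep medium female player body, set up for its dangling front and back strap bones.
# Front left and right strap chains hang off spine_01, while back left, right and center chains hang off spine_05.
# Strap bases have free angular motion, with tight twist and asymmetric swing limits down each chain.
# Larger capsules on pelvis, spine, thighs and left hand give the body volume around the straps.
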PhysicsAsset /BRCosmetics/Characters/Player/Female/Medium/Bodies/F_MED_LooseCreep/Meshes/F_MED_LooseCreep_Physics
bodies {
 dyn_strap_bk_c_01 [{shapeType=Capsule, length=4, radius=1, shapeName=dyn_strap_bk_c_01_capsule, center=<4 0 0>, rotation=<rot 90 0 0>}]
 dyn_strap_bk_c_base [{shapeType=Capsule, length=2, radius=1, shapeName=dyn_strap_bk_c_base_capsule}]
 dyn_strap_bk_l_01 [{shapeType=Capsule, length=2, radius=1, shapeName=dyn_strap_bk_l_01_capsule, center=<2 0 0>, rotation=<rot 90 0 0>}]
 dyn_strap_bk_l_02 [{shapeType=Capsule, length=2, radius=1, shapeName=dyn_strap_bk_l_02_capsule, center=<2 0 0>, rotation=<rot 90 0 0>}]
 dyn_strap_bk_l_base [{shapeType=Capsule, length=2, radius=1, shapeName=dyn_strap_bk_l_base_capsule, center=<2 0 0>, rotation=<rot 90 0 0>}]
 dyn_strap_bk_r_01 [{shapeType=Capsule, length=2, radius=1, shapeName=dyn_strap_bk_r_01_capsule, center=<2 0 0>, rotation=<rot 90 0 0>}]
 dyn_strap_bk_r_02 [{shapeType=Capsule, length=2, radius=1, shapeName=dyn_strap_bk_r_02_capsule, center=<2 0 0>, rotation=<rot 90 0 0>}]
 dyn_strap_bk_r_base [{shapeType=Capsule, length=2, radius=1, shapeName=dyn_strap_bk_r_base_capsule}]
 dyn_strap_fr_l_01 [{shapeType=Capsule, length=2, radius=1, shapeName=dyn_strap_fr_l_01_capsule, center=<2 0 0>, rotation=<rot 90 0 0>}]
 dyn_strap_fr_l_02 [{shapeType=Capsule, length=2, radius=1, shapeName=dyn_strap_fr_l_02_capsule, center=<2 0 0>, rotation=<rot 90 0 0>}]
 dyn_strap_fr_l_03 [{shapeType=Capsule, length=2, radius=1, shapeName=dyn_strap_fr_l_03_capsule, center=<2 0 0.5>, rotation=<rot 0 0 90>}]
 dyn_strap_fr_l_base [{shapeType=Capsule, length=2, radius=1, shapeName=dyn_strap_fr_l_base_capsule}]
 dyn_strap_fr_r_01 [{shapeType=Capsule, length=2, radius=1, shapeName=dyn_strap_fr_r_01_capsule, center=<2 0 0>, rotation=<rot 90 0 0>}]
 dyn_strap_fr_r_02 [{shapeType=Capsule, length=2, radius=1, shapeName=dyn_strap_fr_r_02_capsule, center=<2 0 0>, rotation=<rot 90 0 0>}]
 dyn_strap_fr_r_03 [{shapeType=Capsule, length=2, radius=1, shapeName=dyn_strap_fr_r_03_capsule, center=<2 0 0.5>, rotation=<rot 0 0 90>}]
 dyn_strap_fr_r_base [{shapeType=Capsule, length=2, radius=1, shapeName=dyn_strap_fr_r_base_capsule}]
 hand_l [{shapeType=Capsule, length=4.781, radius=6.8, shapeName=hand_l_capsule, center=<5 -0.6495 -0.2806>, rotation=<rot 0 -90 0>}]
 pelvis [{shapeType=Capsule, length=8, radius=9}]
 spine_01 [{shapeType=Capsule, length=6, radius=8, shapeName=spine_01_capsule}]
 spine_03 [{shapeType=Capsule, length=8, radius=10, shapeName=spine_03_capsule}]
 spine_04 [{shapeType=Capsule, length=8, radius=8, shapeName=spine_04_capsule}]
 spine_05 [{shapeType=Capsule, length=8, radius=7, shapeName=spine_05_capsule, center=<0.3599 -11.3305 0>}]
 thigh_l [{shapeType=Capsule, length=21.259, radius=8, center=<-16.9476 0.5903 0.3174>, rotation=<rot -0.6426 -99.8569 86.3528>}]
 thigh_r [{shapeType=Capsule, length=21.259, radius=8, center=<16.9476 -0.5903 -0.3174>, rotation=<rot -0.3225 -94.8668 -93.6892>}]
 thigh_twist_01_l [{
   shapeType=Capsule
   length=20.3505
   radius=7
   shapeName=thigh_twist_01_l_capsule
   center=<8.94 0.5094 0.5787>
   rotation=<rot 0.0001 -89.8778 81.2967>
  }]
}
constraints [
 {bone1Name=thigh_l, bone2Name=pelvis, swing1LimitDegrees=45, swing2LimitDegrees=45, twistLimitDegrees=45}
 {bone1Name=thigh_r, bone2Name=pelvis, swing1LimitDegrees=45, swing2LimitDegrees=45, twistLimitDegrees=45}
 {
  bone1Name=spine_01
  bone2Name=pelvis
  swing1LimitDegrees=45
  swing1Motion=Free
  swing2LimitDegrees=45
  swing2Motion=Free
  twistLimitDegrees=45
  twistMotion=Free
 }
 {
  bone1Name=spine_03
  bone2Name=spine_01
  swing1LimitDegrees=45
  swing1Motion=Free
  swing2LimitDegrees=45
  swing2Motion=Free
  twistLimitDegrees=45
  twistMotion=Free
 }
 {
  bone1Name=spine_04
  bone2Name=spine_03
  swing1LimitDegrees=45
  swing1Motion=Free
  swing2LimitDegrees=45
  swing2Motion=Free
  twistLimitDegrees=45
  twistMotion=Free
 }
 {
  bone1Name=dyn_strap_fr_l_base
  bone2Name=spine_01
  swing1LimitDegrees=45
  swing1Motion=Free
  swing2LimitDegrees=45
  swing2Motion=Free
  twistLimitDegrees=45
  twistMotion=Free
 }
 {bone1Name=dyn_strap_fr_l_01, bone2Name=dyn_strap_fr_l_base, swing1LimitDegrees=5, swing2LimitDegrees=30, twistLimitDegrees=5}
 {bone1Name=dyn_strap_fr_l_02, bone2Name=dyn_strap_fr_l_01, swing1LimitDegrees=10, swing2LimitDegrees=30, twistLimitDegrees=5}
 {bone1Name=dyn_strap_fr_l_03, bone2Name=dyn_strap_fr_l_02, swing1LimitDegrees=15, swing2LimitDegrees=45, twistLimitDegrees=5}
 {
  bone1Name=dyn_strap_fr_r_base
  bone2Name=spine_01
  swing1LimitDegrees=45
  swing1Motion=Free
  swing2LimitDegrees=45
  swing2Motion=Free
  twistLimitDegrees=45
  twistMotion=Free
 }
 {bone1Name=dyn_strap_fr_r_01, bone2Name=dyn_strap_fr_r_base, swing1LimitDegrees=15, swing2LimitDegrees=30, twistLimitDegrees=5}
 {bone1Name=dyn_strap_fr_r_02, bone2Name=dyn_strap_fr_r_01, swing1LimitDegrees=15, swing2LimitDegrees=30, twistLimitDegrees=5}
 {bone1Name=dyn_strap_fr_r_03, bone2Name=dyn_strap_fr_r_02, swing1LimitDegrees=15, swing2LimitDegrees=60, twistLimitDegrees=5}
 {bone1Name=dyn_strap_bk_r_02, bone2Name=dyn_strap_bk_r_01, swing1LimitDegrees=15, swing2LimitDegrees=30, twistLimitDegrees=5}
 {bone1Name=dyn_strap_bk_r_01, bone2Name=dyn_strap_bk_r_base, swing1LimitDegrees=15, swing2LimitDegrees=30, twistLimitDegrees=5}
 {
  bone1Name=dyn_strap_bk_r_base
  bone2Name=spine_05
  swing1LimitDegrees=45
  swing1Motion=Free
  swing2LimitDegrees=45
  swing2Motion=Free
  twistLimitDegrees=45
  twistMotion=Free
 }
 {bone1Name=dyn_strap_bk_l_02, bone2Name=dyn_strap_bk_l_01, swing1LimitDegrees=15, swing2LimitDegrees=30, twistLimitDegrees=5}
 {bone1Name=dyn_strap_bk_l_01, bone2Name=dyn_strap_bk_l_base, swing1LimitDegrees=15, swing2LimitDegrees=30, twistLimitDegrees=5}
 {
  bone1Name=dyn_strap_bk_l_base
  bone2Name=spine_05
  swing1LimitDegrees=45
  swing1Motion=Free
  swing2LimitDegrees=45
  swing2Motion=Free
  twistLimitDegrees=45
  twistMotion=Free
 }
 {bone1Name=dyn_strap_bk_c_01, bone2Name=dyn_strap_bk_c_base, swing1LimitDegrees=15, swing2LimitDegrees=30, twistLimitDegrees=5}
 {
  bone1Name=dyn_strap_bk_c_base
  bone2Name=spine_05
  swing1LimitDegrees=45
  swing1Motion=Free
  swing2LimitDegrees=45
  swing2Motion=Free
  twistLimitDegrees=45
  twistMotion=Free
 }
 {
  bone1Name=thigh_twist_01_l
  bone2Name=thigh_l
  swing1LimitDegrees=45
  swing1Motion=Free
  swing2LimitDegrees=45
  swing2Motion=Free
  twistLimitDegrees=45
  twistMotion=Free
 }
 {
  bone1Name=spine_05
  bone2Name=spine_04
  swing1LimitDegrees=45
  swing1Motion=Free
  swing2LimitDegrees=45
  swing2Motion=Free
  twistLimitDegrees=45
  twistMotion=Free
 }
 {
  bone1Name=hand_l
  bone2Name=spine_05
  swing1LimitDegrees=45
  swing1Motion=Free
  swing2LimitDegrees=45
  swing2Motion=Free
  twistLimitDegrees=45
  twistMotion=Free
 }
]
