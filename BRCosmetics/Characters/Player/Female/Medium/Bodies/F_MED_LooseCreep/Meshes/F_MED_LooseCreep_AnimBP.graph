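# Animation Blueprint for the LooseCreep female medium body cosmetic, copying the attached parent mesh's pose and adding secondary physics motion.
# The pose runs through a pre-dynamics Control Rig, a RigidBody simulation in spine_04 bone space, then a post-dynamics Control Rig.
# Each update computes a gravity override vector from a parameter asset through a shared function and flags whether the current LOD is 2 or higher.
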AnimBlueprint /BRCosmetics/Characters/Player/Female/Medium/Bodies/F_MED_LooseCreep/Meshes/F_MED_LooseCreep_AnimBP
target_skeleton=&/F_MED_LooseCreep_Skeleton
graphs [
 AnimGraph {nodes [
   n1 = Control Rig!(Source=Copy Pose From Mesh() {bUseAttachedParent=true}, Alpha, target:bool=false, LOD 2-3:bool=false) -> (Pose) {controlRigAssetReference {blueprintRigClass=&/F_MED_LooseCreep_CtrlRig_PreDyn.F_MED_LooseCreep_CtrlRig_PreDyn_C}}
   n2 = RigidBody!(
    ComponentPose=Local To Component(LocalPose=@n1)
    bAlphaBoolEnabled:bool
    OverrideWorldGravity:Vector="0,0,0"
    ComponentLinearAccScale:Vector="0,0,0"
    ComponentLinearVelScale:Vector="0,0,0"
    bTransferBoneVelocities:bool=False
   ) -> (Pose) {
    alphaInputType=Bool
    bClampLinearTranslationLimitToRefPose=true
    bFreezeIncomingPoseOnStart=true
    bOverrideWorldGravity=true
    bTransferBoneVelocities=true
    overridePhysicsAsset=&/F_MED_LooseCreep_Physics
    simulationSpace=BaseBoneSpace
    alphaBoolBlend {blendOption=Linear}
    baseBoneRef {boneName=spine_04}
    componentAppliedLinearAccClamp=<100 100 100>
    componentLinearAccScale=<0.6 0.2 0.2>
    componentLinearVelScale=<0.6 0.2 0.2>
    simSpaceSettings {
     dampingAlpha=0
     maxAngularAcceleration=50
     maxAngularVelocity=8
     maxLinearAcceleration=2000
     maxLinearVelocity=500
     worldAlpha=1
     externalLinearDragV=<1 1 1>
    }
   }
   n3 = Control Rig!(Source=Component To Local(ComponentPose=@n2), Alpha, FE:bool=false, Emote:bool=false) -> (Pose) {controlRigAssetReference {blueprintRigClass=&/F_MED_LooseCreep_CtrlRig_PostDyn.F_MED_LooseCreep_CtrlRig_PostDyn_C}}
   Output Pose(Result=@n3) {layerGroup=None}
  ]}
 EventGraph {nodes [
   Event BlueprintUpdateAnimation()
   > n1 = Sequence()
   Set LOD 2-3(execute=@n1.then_1, LOD 2-3:bool="integer >= integer"(A=Get CurrentLOD(), B=2)) -> (Output_Get:bool)
   n2 = "Gravity Override State Function (Parameter Asset) (Nativized)"(
    execute=@n1.then_0
    CharacterPartAnimInstance=Self-Reference()
    GameTimeInSeconds=GetGameTimeInSeconds()
    DebugJointName=spine_05
    Parameters=&/F_MED_LooseCreep_Parameters
   )
   > Set GravityOverride(GravityOverride=@n2)
  ]}
]
variables [
 GravityOverride {type=Vector}
 LOD 2-3 {type=boolean}
]
