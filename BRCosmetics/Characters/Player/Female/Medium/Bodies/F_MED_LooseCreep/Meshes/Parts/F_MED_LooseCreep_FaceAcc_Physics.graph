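# Physics asset for the LooseCreep female medium body's FaceAcc part, giving its dynamic hair bones simulation bodies around the head.
# Front, back and side strands form short capsule chains of two or three bones, each rooted on a head body built from two large capsules.
# Base-to-head joints leave swing and twist free, while joints further along each strand carry tighter angle limits, especially on twist.
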PhysicsAsset /BRCosmetics/Characters/Player/Female/Medium/Bodies/F_MED_LooseCreep/Meshes/Parts/F_MED_LooseCreep_FaceAcc_Physics
bodies {
 dyn_hair_01_fr_01 [{shapeType=Capsule, length=2, radius=1.5, shapeName=dyn_hair_01_fr_01_capsule, center=<2 0 0>, rotation=<rot 90 0 0>}]
 dyn_hair_01_fr_02 [{shapeType=Capsule, length=2, radius=1, shapeName=dyn_hair_01_fr_02_capsule, center=<2 0 0>, rotation=<rot 90 0 0>}]
 dyn_hair_01_fr_base [{shapeType=Capsule, length=2, radius=0.5, shapeName=dyn_hair_01_fr_base_capsule}]
 dyn_hair_02_fr_01 [{shapeType=Capsule, length=2, radius=1.5, shapeName=dyn_hair_02_fr_01_capsule, center=<2 0 0.5>, rotation=<rot 90 0 0>}]
 dyn_hair_02_fr_02 [{shapeType=Capsule, length=2, radius=1.5, shapeName=dyn_hair_02_fr_02_capsule, center=<2 0 0.5>, rotation=<rot 90 0 0>}]
 dyn_hair_02_fr_base [{shapeType=Capsule, length=2, radius=0.5, shapeName=dyn_hair_02_fr_base_capsule}]
 dyn_hair_03_fr_01 [{shapeType=Capsule, length=2, radius=1, shapeName=dyn_hair_03_fr_01_capsule, center=<2 0 0>, rotation=<rot 90 0 0>}]
 dyn_hair_03_fr_02 [{shapeType=Capsule, length=2, radius=2, shapeName=dyn_hair_03_fr_02_capsule, center=<2 0 1>, rotation=<rot 90 0 0>}]
 dyn_hair_03_fr_base [{shapeType=Capsule, length=2, radius=0.5, shapeName=dyn_hair_03_fr_base_capsule}]
 dyn_hair_bk_01 [{shapeType=Capsule, length=2, radius=1, shapeName=dyn_hair_bk_01_capsule, center=<2 0 0>, rotation=<rot 90 0 0>}]
 dyn_hair_bk_01_base [{shapeType=Capsule, length=2, radius=0.5, shapeName=dyn_hair_bk_01_base_capsule}]
 dyn_hair_bk_02 [{shapeType=Capsule, length=2, radius=1, shapeName=dyn_hair_bk_02_capsule, center=<2 0 0>, rotation=<rot 90 0 0>}]
 dyn_hair_bk_02_base [{shapeType=Capsule, length=2, radius=0.5, shapeName=dyn_hair_bk_02_base_capsule}]
 dyn_hair_bk_03 [{shapeType=Capsule, length=2, radius=1, shapeName=dyn_hair_bk_03_capsule, center=<2 0 0>, rotation=<rot 90 0 0>}]
 dyn_hair_bk_03_base [{shapeType=Capsule, length=2, radius=0.5, shapeName=dyn_hair_bk_03_base_capsule}]
 dyn_hair_bk_04 [{shapeType=Capsule, length=2, radius=1, shapeName=dyn_hair_bk_04_capsule, center=<2 0 0>, rotation=<rot 90 0 0>}]
 dyn_hair_bk_04_base [{shapeType=Capsule, length=2, radius=0.5, shapeName=dyn_hair_bk_04_base_capsule}]
 dyn_hair_bk_05 [{shapeType=Capsule, length=2, radius=1, shapeName=dyn_hair_bk_05_capsule, center=<2 0 0>, rotation=<rot 90 0 0>}]
 dyn_hair_bk_05_base [{shapeType=Capsule, length=2, radius=0.5, shapeName=dyn_hair_bk_05_base_capsule}]
 dyn_hair_bk_06 [{shapeType=Capsule, length=2, radius=1, shapeName=dyn_hair_bk_06_capsule, center=<2 0 0>, rotation=<rot 90 0 0>}]
 dyn_hair_bk_06_base [{shapeType=Capsule, length=2, radius=0.5, shapeName=dyn_hair_bk_06_base_capsule}]
 dyn_hair_bk_07 [{shapeType=Capsule, length=2, radius=1, shapeName=dyn_hair_bk_07_capsule, center=<2 0 0>, rotation=<rot 90 0 0>}]
 dyn_hair_bk_07_base [{shapeType=Capsule, length=2, radius=0.5, shapeName=dyn_hair_bk_07_base_capsule}]
 dyn_hair_bk_08 [{shapeType=Capsule, length=2, radius=1, shapeName=dyn_hair_bk_08_capsule, center=<2 0 0>, rotation=<rot 90 0 0>}]
 dyn_hair_bk_08_base [{shapeType=Capsule, length=2, radius=0.5, shapeName=dyn_hair_bk_08_base_capsule}]
 dyn_hair_side_01_r [{shapeType=Capsule, length=2, radius=1, shapeName=dyn_hair_side_01_r_capsule, center=<2 0 0>, rotation=<rot 90 0 0>}]
 dyn_hair_side_01_r_base [{shapeType=Capsule, length=2, radius=0.5, shapeName=dyn_hair_side_01_r_base_capsule}]
 dyn_hair_side_02_r [{shapeType=Capsule, length=2, radius=1, shapeName=dyn_hair_side_02_r_capsule, center=<2 0 0>, rotation=<rot 90 0 0>}]
 dyn_hair_side_02_r_base [{shapeType=Capsule, length=2, radius=0.5, shapeName=dyn_hair_side_02_r_base_capsule}]
 dyn_hair_side_l_01 [{shapeType=Capsule, length=2, radius=1, shapeName=dyn_hair_side_l_01_capsule, center=<2 0 0>, rotation=<rot 90 0 0>}]
 dyn_hair_side_l_02 [{shapeType=Capsule, length=2, radius=1, shapeName=dyn_hair_side_l_02_capsule, center=<2 0 0>, rotation=<rot 90 0 0>}]
 dyn_hair_side_l_base [{shapeType=Capsule, length=2, radius=0.5, shapeName=dyn_hair_side_l_base_capsule}]
 head [
  {shapeType=Capsule, length=10, radius=8, rotation=<rot 90 0 0>}
  {shapeType=Capsule, length=10, radius=7, shapeName=head_capsule, center=<0.775 2.3036 0>, rotation=<rot 90 0 0>}
 ]
}
constraints [
 {
  bone1Name=dyn_hair_01_fr_base
  bone2Name=head
  swing1LimitDegrees=45
  swing1Motion=Free
  swing2LimitDegrees=45
  swing2Motion=Free
  twistLimitDegrees=45
  twistMotion=Free
 }
 {bone1Name=dyn_hair_01_fr_01, bone2Name=dyn_hair_01_fr_base, swing1LimitDegrees=15, swing2LimitDegrees=30, twistLimitDegrees=5}
 {bone1Name=dyn_hair_01_fr_02, bone2Name=dyn_hair_01_fr_01, swing1LimitDegrees=15, swing2LimitDegrees=30, twistLimitDegrees=5}
 {
  bone1Name=dyn_hair_02_fr_base
  bone2Name=head
  swing1LimitDegrees=45
  swing1Motion=Free
  swing2LimitDegrees=45
  swing2Motion=Free
  twistLimitDegrees=45
  twistMotion=Free
 }
 {bone1Name=dyn_hair_02_fr_01, bone2Name=dyn_hair_02_fr_base, swing1LimitDegrees=15, swing2LimitDegrees=30, twistLimitDegrees=5}
 {bone1Name=dyn_hair_02_fr_02, bone2Name=dyn_hair_02_fr_01, swing1LimitDegrees=15, swing2LimitDegrees=30, twistLimitDegrees=5}
 {
  bone1Name=dyn_hair_03_fr_base
  bone2Name=head
  swing1LimitDegrees=45
  swing1Motion=Free
  swing2LimitDegrees=45
  swing2Motion=Free
  twistLimitDegrees=45
  twistMotion=Free
 }
 {bone1Name=dyn_hair_03_fr_01, bone2Name=dyn_hair_03_fr_base, swing1LimitDegrees=15, swing2LimitDegrees=30, twistLimitDegrees=5}
 {bone1Name=dyn_hair_03_fr_02, bone2Name=dyn_hair_03_fr_01, swing1LimitDegrees=15, swing2LimitDegrees=30, twistLimitDegrees=5}
 {
  bone1Name=dyn_hair_bk_01_base
  bone2Name=head
  swing1LimitDegrees=45
  swing1Motion=Free
  swing2LimitDegrees=45
  swing2Motion=Free
  twistLimitDegrees=45
  twistMotion=Free
 }
 {bone1Name=dyn_hair_bk_01, bone2Name=dyn_hair_bk_01_base, swing1LimitDegrees=15, swing2LimitDegrees=30, twistLimitDegrees=5}
 {
  bone1Name=dyn_hair_bk_02_base
  bone2Name=head
  swing1LimitDegrees=45
  swing1Motion=Free
  swing2LimitDegrees=45
  swing2Motion=Free
  twistLimitDegrees=45
  twistMotion=Free
 }
 {bone1Name=dyn_hair_bk_02, bone2Name=dyn_hair_bk_02_base, swing1LimitDegrees=15, swing2LimitDegrees=30, twistLimitDegrees=5}
 {
  bone1Name=dyn_hair_bk_03_base
  bone2Name=head
  swing1LimitDegrees=45
  swing1Motion=Free
  swing2LimitDegrees=45
  swing2Motion=Free
  twistLimitDegrees=45
  twistMotion=Free
 }
 {bone1Name=dyn_hair_bk_03, bone2Name=dyn_hair_bk_03_base, swing1LimitDegrees=15, swing2LimitDegrees=30, twistLimitDegrees=5}
 {
  bone1Name=dyn_hair_bk_04_base
  bone2Name=head
  swing1LimitDegrees=45
  swing1Motion=Free
  swing2LimitDegrees=45
  swing2Motion=Free
  twistLimitDegrees=45
  twistMotion=Free
 }
 {bone1Name=dyn_hair_bk_04, bone2Name=dyn_hair_bk_04_base, swing1LimitDegrees=15, swing2LimitDegrees=30, twistLimitDegrees=5}
 {
  bone1Name=dyn_hair_bk_05_base
  bone2Name=head
  swing1LimitDegrees=45
  swing1Motion=Free
  swing2LimitDegrees=45
  swing2Motion=Free
  twistLimitDegrees=45
  twistMotion=Free
 }
 {bone1Name=dyn_hair_bk_05, bone2Name=dyn_hair_bk_05_base, swing1LimitDegrees=15, swing2LimitDegrees=30, twistLimitDegrees=5}
 {
  bone1Name=dyn_hair_bk_06_base
  bone2Name=head
  swing1LimitDegrees=45
  swing1Motion=Free
  swing2LimitDegrees=45
  swing2Motion=Free
  twistLimitDegrees=45
  twistMotion=Free
 }
 {bone1Name=dyn_hair_bk_06, bone2Name=dyn_hair_bk_06_base, swing1LimitDegrees=15, swing2LimitDegrees=30, twistLimitDegrees=5}
 {
  bone1Name=dyn_hair_bk_07_base
  bone2Name=head
  swing1LimitDegrees=45
  swing1Motion=Free
  swing2LimitDegrees=45
  swing2Motion=Free
  twistLimitDegrees=45
  twistMotion=Free
 }
 {bone1Name=dyn_hair_bk_07, bone2Name=dyn_hair_bk_07_base, swing1LimitDegrees=15, swing2LimitDegrees=30, twistLimitDegrees=5}
 {
  bone1Name=dyn_hair_bk_08_base
  bone2Name=head
  swing1LimitDegrees=45
  swing1Motion=Free
  swing2LimitDegrees=45
  swing2Motion=Free
  twistLimitDegrees=45
  twistMotion=Free
 }
 {bone1Name=dyn_hair_bk_08, bone2Name=dyn_hair_bk_08_base, swing1LimitDegrees=15, swing2LimitDegrees=30, twistLimitDegrees=5}
 {
  bone1Name=dyn_hair_side_01_r_base
  bone2Name=head
  swing1LimitDegrees=45
  swing1Motion=Free
  swing2LimitDegrees=45
  swing2Motion=Free
  twistLimitDegrees=45
  twistMotion=Free
 }
 {bone1Name=dyn_hair_side_01_r, bone2Name=dyn_hair_side_01_r_base, swing1LimitDegrees=15, swing2LimitDegrees=30, twistLimitDegrees=5}
 {
  bone1Name=dyn_hair_side_02_r_base
  bone2Name=head
  swing1LimitDegrees=45
  swing1Motion=Free
  swing2LimitDegrees=45
  swing2Motion=Free
  twistLimitDegrees=45
  twistMotion=Free
 }
 {bone1Name=dyn_hair_side_02_r, bone2Name=dyn_hair_side_02_r_base, swing1LimitDegrees=15, swing2LimitDegrees=30, twistLimitDegrees=5}
 {
  bone1Name=dyn_hair_side_l_base
  bone2Name=head
  swing1LimitDegrees=45
  swing1Motion=Free
  swing2LimitDegrees=45
  swing2Motion=Free
  twistLimitDegrees=45
  twistMotion=Free
 }
 {bone1Name=dyn_hair_side_l_01, bone2Name=dyn_hair_side_l_base, swing1LimitDegrees=15, swing2LimitDegrees=30, twistLimitDegrees=5}
 {bone1Name=dyn_hair_side_l_02, bone2Name=dyn_hair_side_l_01, swing1LimitDegrees=15, swing2LimitDegrees=30, twistLimitDegrees=5}
]
